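# Translucent, unlit, two-sided Niagara sprite and mesh material for the JadeTowel Gloss emote's circular backdrop effect.
# Two panning noise samples in radial coordinates distort a soft circle, which drives a curve-atlas color, hue-shifted by a dynamic parameter and modulated by particle color and rainbow texture.
# Opacity combines particle alpha, the distorted circle, an AppearMask-sized reveal and view-angle fade, with depth fade only on default and deferred shading paths.
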Material /BRCosmetics/Animation/Game/MainPlayer/Emotes/JadeTowel_Gloss/FX/M_Emote_JadeTowelGloss_Backdrop
graph {
 properties {
  bAutomaticallySetUsageInEditor=false
  bUsedWithNiagaraMeshParticles=true
  bUsedWithNiagaraSprites=true
  blendMode=BLEND_Translucent
  shadingModel=MSM_Unlit
  twoSided=true
  usedShadingModels=Unlit
 }
 nodes [
  n1 = Multiply(
   A=MaterialFunctionCall() {materialFunction=&VectorToRadialValue}.Radial Coordinates
   B=AppendVector(A=Constant() {r=8}, B=Constant())
  )
  n2 = TextureSample!(UVs=Panner(Coordinate=@n1) {speedX=0.3, speedY=-0.3}, Tex, Level, Apply View MipBias) {mipValueMode=TMVM_MipLevel, texture=/Game/Effects/Fort_Effects/Textures/Tile/Tests/T_Tiling_PanningMask_Softest_Seamless}
  n3 = TextureSample(UVs=Panner(Coordinate=@n1) {speedY=-1.5}) {mipValueMode=TMVM_MipLevel, texture=&T_Tiling_PanningMask_Softest_Seamless}
  n4 = Add(A=Multiply(A=Subtract(A=@n2.R) {constB=0.5}) {constB=2}, B=@n3.R)
  n5 = AppendVector(A=Multiply(A=@n4) {constB=0.07}, B=Multiply(A=@n4) {constB=-0.07})
  n6 = Add(A=Add(A=TextureCoordinate(), B=Constant2Vector()), B=@n5)
  n7 = Constant() {r=0.5}
  n8 = Constant() {r=0.45}
  n9 = Constant()
  n10 = CurveAtlasRowParameter(CurveTime=Power(Base=SphereMask(A=@n6, B=@n7, Radius=@n8, Hardness=@n9)) {constExponent=1}) {
   atlas=/BRCosmetics/Effects/Fort_Effects/Emotes/WarmShade/Materials/CAtlas_Emote_WarmShade
   curve=&/CV_Emote_JadeTowelGloss_Color1
   defaultValue=6
   parameterName=ColorCurve
  }
  n11 = Multiply(A=LinearInterpolate(Alpha=@n3.R) {constA=0.5})
  n12 = ParticleColor()
  n13 = Desaturation(
   None=TextureSample!(UVs=TextureCoordinate() {uTiling=1.2, vTiling=1.2}, Tex, Apply View MipBias) {texture=&T_Tile_RainbowColor}.RGB
   Fraction=Constant() {r=-0.5}
  )
  n14 = Multiply(
   A=MaterialFunctionCall(Hue Shift Percentage=DynamicParameter() {defaultValue=<rgba 1 0 1 1>, paramNames [Param1,Color,AppearMask,Param4]}.Color, Texture=@n10) {materialFunction=&HueShift}.Result
   B=Add(A=Multiply(A=@n11, B=@n12.RGB), B=LinearInterpolate(B=@n13) {constAlpha=0.4})
  )
  n15 = ScalarParameter() {parameterName=Desaturation}
  MaterialFunctionCall(Emissive=Multiply(A=Desaturation(None=@n14, Fraction=@n15), B=ScalarParameter() {defaultValue=1, parameterName=All Light})) {materialFunction=&MF_FX_EmissiveAdjustments}
  n16 = TextureCoordinate()
  n17 = Constant2Vector()
  n18 = SphereMask(A=Add(A=Add(A=@n16, B=@n17), B=@n5), B=Constant() {r=0.5}, Radius=Constant() {r=0.45}, Hardness=Constant() {r=0.5})
  n19 = TextureCoordinate()
  n20 = Constant() {r=0.5}
  n21 = DynamicParameter() {defaultValue=<rgba 1 0 1 1>, paramNames [Param1,Color,AppearMask,Param4]}
  n22 = Constant() {r=0.5}
  n23 = Multiply(
   A=Multiply(A=Multiply(A=@n12.A, B=@n18), B=SphereMask(A=@n19, B=@n20, Radius=@n21.AppearMask, Hardness=@n22))
   B=MaterialFunctionCall("FadeSensitivity (optional)"=ScalarParameter() {parameterName=MeshFade}) {materialFunction=&FadeBasedOnViewAngle}.Result
  )
  n24 = DepthFade(Opacity=@n23, FadeDistance=ScalarParameter() {defaultValue=5, parameterName=DepthFade})
  ShadingPathSwitch(Default=@n24, Deferred=@n24, Forward=@n23, Mobile=@n23)
 ]
}
