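# Translucent unlit two-sided Niagara material for the ring line effect in the JadeTowel Gloss emote.
# SplineThicken gets a time-sine width modulated by two opposing panned cell-noise samples, with dynamic parameters for intensity, UV scale and speed.
# Emissive blends particle color toward a hue-shifted curve-atlas gradient sampled from radial world-position coordinates.
# Opacity combines particle alpha, a rotated panning texture and view-angle fade, adding depth fade except on forward and mobile paths.
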Material /BRCosmetics/Animation/Game/MainPlayer/Emotes/JadeTowel_Gloss/FX/M_Emote_JadeTowelGloss_RingLine
graph {
 properties {
  bAutomaticallySetUsageInEditor=false
  bUsedWithNiagaraMeshParticles=true
  bUsedWithNiagaraRibbons=true
  bUsedWithNiagaraSprites=true
  blendMode=BLEND_Translucent
  shadingModel=MSM_Unlit
  twoSided=true
  usedShadingModels=Unlit
 }
 nodes [
  n1 = Sine(None=Time())
  n2 = DynamicParameter() {paramNames [SplineWidth_Intensity, SplineNoiseUV_Offset, SplineNoiseSpeed_Offset, color]}
  n3 = TextureCoordinate() {uTiling=8}
  n4 = DynamicParameter() {paramNames [SplineWidth_Intensity, SplineNoiseUV_Offset, SplineNoiseSpeed_Offset, color]}
  n5 = AppendVector(A=Constant() {r=0.5}, B=Constant() {r=-0.2})
  n6 = TextureSample(UVs=Panner(Coordinate=Multiply(A=@n3, B=@n4.SplineNoiseUV_Offset), Speed=Multiply(A=@n5, B=@n4.SplineNoiseSpeed_Offset)) {speedX=0.5, speedY=-0.2}) {samplerType=SAMPLERTYPE_Grayscale, texture=/BRCosmetics/Weapons/FORT_Melee/Pickaxe_Cube_Ninja_Male/FX/T_CellNoise}
  n7 = TextureCoordinate() {uTiling=2}
  n8 = AppendVector(A=Constant() {r=-0.3}, B=Constant() {r=0.2})
  n9 = TextureSample(UVs=Panner(Coordinate=Multiply(A=@n7, B=@n4.SplineNoiseUV_Offset), Speed=Multiply(A=@n8, B=@n4.SplineNoiseSpeed_Offset)) {speedX=-0.3, speedY=0.2}) {samplerType=SAMPLERTYPE_Grayscale, texture=&T_CellNoise}
  n10 = Multiply(
   A=Multiply(A=ScalarParameter() {defaultValue=8, parameterName=Spline_Base}, B=Multiply(A=@n1, B=@n2.SplineWidth_Intensity) {constB=0})
   B=Power(Base=Add(A=@n6.R, B=@n9.R)) {constExponent=1.2}
  )
  MaterialFunctionCall(WidthBase=@n10, WidthTip=@n10, Expand U or V UV Channel=StaticBool()) {materialFunction=&SplineThicken}
  n11 = ParticleColor()
  n12 = WorldPosition()
  n13 = CurveAtlasRowParameter(CurveTime=ComponentMask(None=MaterialFunctionCall(Vector or UVs=@n12.XYZ) {materialFunction=&VectorToRadialValue}.Radial Coordinates) {g=true}) {
   atlas=/BRCosmetics/Effects/Fort_Effects/Emotes/WarmShade/Materials/CAtlas_Emote_WarmShade
   curve=&/CV_Emote_JadeTowelGloss_Color2
   defaultValue=7
   parameterName=ColorCurve
  }
  n14 = Multiply(
   A=MaterialFunctionCall(
    Hue Shift Percentage=DynamicParameter() {paramNames [SplineWidth_Intensity, SplineNoiseUV_Offset, SplineNoiseSpeed_Offset, color]}.color
    Texture=@n13
   ) {materialFunction=&HueShift}.Result
   B=@n11.RGB
  )
  n15 = ScalarParameter() {parameterName=Use_HueShiftColor}
  MaterialFunctionCall(Emissive=Multiply(A=LinearInterpolate(A=@n11.RGB, B=@n14, Alpha=@n15), B=ScalarParameter() {defaultValue=1, parameterName=All_Light})) {materialFunction=&MF_FX_EmissiveAdjustments}
  ScalarParameter() {defaultValue=20, parameterName=Spline_Tip}
  n16 = Multiply(
   A=TextureCoordinate()
   B=AppendVector(A=ScalarParameter() {defaultValue=100, parameterName=BaseTextureU}, B=ScalarParameter() {defaultValue=1, parameterName=BaseTextureV})
  )
  n17 = AppendVector(A=ScalarParameter() {parameterName=BaseTexture_SpeedX}, B=ScalarParameter() {parameterName=BaseTexture_SpeedY})
  n18 = TextureSampleParameter2D(UVs=MaterialFunctionCall(UVs=Panner(Coordinate=@n16, Speed=@n17), "Rotation Angle (0-1)"=Constant() {r=0.25}) {materialFunction=&CustomRotator}.Rotated Values) {parameterName=BaseTexture, texture=/Game/Global/Textures/Default/Masks_and_Gradients/gradient_stasis}
  n19 = ScalarParameter() {defaultValue=1, parameterName=BaseTexturePower}
  n20 = Multiply(
   A=Multiply(A=@n11.A, B=Power(Base=@n18.R, Exp=@n19))
   B=MaterialFunctionCall("FadeSensitivity (optional)"=ScalarParameter() {parameterName=MeshFade}) {materialFunction=&FadeBasedOnViewAngle}.Result
  )
  n21 = DepthFade(Opacity=@n20, FadeDistance=ScalarParameter() {defaultValue=5, parameterName=DepthFade})
  ShadingPathSwitch(Default=@n21, Deferred=@n21, Forward=@n20, Mobile=@n20)
  TextureCoordinate()
 ]
}
