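# Unlit, two-sided translucent Niagara sprite/mesh material for JadeTowel Gloss emote FX.
# A star texture, scaled by dynamic parameter, blends by scalar toggles with panning noise or stepped radial masks, then multiplies screen-space tiled dots.
# Emissive tints this with particle color and a radial curve-atlas gradient, hue-shifted per particle.
# Opacity adds particle alpha, sphere-mask and view-angle fades, plus depth fade except on Forward and Mobile.
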Material /BRCosmetics/Animation/Game/MainPlayer/Emotes/JadeTowel_Gloss/FX/M_Emote_JadeTowelGloss_SpriteBox
graph {
 properties {
  bAutomaticallySetUsageInEditor=false
  bUsedWithNiagaraMeshParticles=true
  bUsedWithNiagaraSprites=true
  blendMode=BLEND_Translucent
  shadingModel=MSM_Unlit
  twoSided=true
  usedShadingModels=Unlit
 }
 nodes [
  n1 = Time()
  n2 = Floor(None=Multiply(A=Sine(None=@n1), B=Constant() {r=8}))
  n3 = TextureSampleParameter2D(UVs=Multiply(A=TextureCoordinate(), B=Multiply(A=@n2) {constB=0.2})) {
   parameterName=MaskTexture01
   samplerType=SAMPLERTYPE_LinearColor
   texture=/Game/TheOutpost/Textures/AnimatedOutpost_Base046_rgb_XVector_a_NormalizedHierPos_UV_2
  }
  n4 = TextureSampleParameter2D(UVs=Panner(Coordinate=TextureCoordinate() {uTiling=2}) {speedX=0.2, speedY=-2}) {
   parameterName=MaskTexture02
   texture=/BRCosmetics/Weapons/FORT_Melee/Pickaxe_Future_Samurai_Summer_Male/FX/T_Pickaxe_Future_Samurai_Summer_Male_Trail_03
  }
  Add(A=LinearInterpolate(Alpha=Power(Base=@n3.G)) {constA=-2}, B=@n4.R)
  Constant() {r=1}
  MaterialFunctionCall() {materialFunction=&VectorToRadialValue}
  n5 = CurveAtlasRowParameter(CurveTime=ComponentMask(None=MaterialFunctionCall() {materialFunction=&VectorToRadialValue}.Radial Coordinates) {g=true}) {
   atlas=/BRCosmetics/Effects/Fort_Effects/Emotes/WarmShade/Materials/CAtlas_Emote_WarmShade
   curve=&/CV_Emote_JadeTowelGloss_Color2
   defaultValue=7
   parameterName=ColorCurve
  }
  n6 = MaterialFunctionCall(Hue Shift Percentage=DynamicParameter() {paramNames [StarUVScale,Color,Param3,Param4]}.Color, Texture=@n5) {materialFunction=&HueShift}
  n7 = MaterialFunctionCall(World Position=WorldPosition().XYZ) {materialFunction=&TransformToClipSpace}
  n8 = Multiply(A=ComponentMask(None=@n7.Buffer UVs for Scene Textures) {r=true}, B=ScalarParameter() {defaultValue=100, parameterName=PointTexture_TilingU}) {constB=100}
  n9 = Multiply(A=ComponentMask(None=@n7.Buffer UVs for Scene Textures) {g=true}, B=ScalarParameter() {defaultValue=50, parameterName=PointTexture_TilingV}) {constB=50}
  n10 = TextureSampleParameter2D(UVs=Multiply(A=AppendVector(A=@n8, B=@n9)) {constB=2}) {parameterName=PointTexture, texture=/Game/Effects/Fort_Effects/Textures/Mask/T_Sphere_A}
  n11 = TextureSampleParameter2D(UVs=MaterialFunctionCall(UVs=TextureCoordinate(), Texture Scale=DynamicParameter() {paramNames [StarUVScale,Color,Param3,Param4]}.StarUVScale) {materialFunction=&ScaleUVsByCenter}.UVs) {
   parameterName=StarTexture
   samplerSource=SSM_Clamp_WorldGroupSettings
   texture=&/T_Emote_JadeTowelGloss_StarRGB
  }
  n12 = ScalarParameter() {parameterName=Use_StarGlowTexture}
  n13 = TextureSampleParameter2D(UVs=Panner(Coordinate=TextureCoordinate() {uTiling=2, vTiling=0.2}) {speedY=2}) {parameterName=MaskTexture03, texture=/Game/Effects/Fort_Effects/Textures/Tile/Tests/T_Tiling_PanningMask_Softest_Seamless}
  n14 = TextureSampleParameter2D(UVs=Panner(Coordinate=TextureCoordinate() {uTiling=2, vTiling=0.2}) {speedY=0.5}) {parameterName=MaskTexture04, samplerType=SAMPLERTYPE_Masks, texture=/Game/Accessories/FORT_Backpacks/M_MED_Heist/Textures/T_CrystalNoise_M}
  n15 = Multiply(A=Time(), B=Constant() {r=8})
  n16 = Add(
   A=ComponentMask(None=MaterialFunctionCall() {materialFunction=&VectorToRadialValue}.Radial Coordinates) {r=true}
   B=Multiply(A=Floor(None=@n15)) {constB=0.3}
  )
  n17 = LinearInterpolate(
   A=LinearInterpolate(A=LinearInterpolate(A=@n11.R, B=@n11.R, Alpha=@n12), B=@n11.B, Alpha=ScalarParameter() {parameterName=Use_StarCenterGlowTexture})
   B=LinearInterpolate(
    A=Multiply(A=@n13.R, B=@n14.R)
    B=TextureSampleParameter2D(UVs=@n16) {parameterName=MaskTexture05, texture=/Game/Athena/Deimos/Spawners/RiftSpawners/Assets/CubeGrowth/T_CubeGrowth_occlusion}.R
    Alpha=ScalarParameter() {parameterName=Use_MaskChangeLine}
   )
   Alpha=ScalarParameter() {parameterName=Use_MaskChange}
  )
  n18 = ParticleColor()
  n19 = Multiply(A=Multiply(A=@n10.R, B=@n17), B=@n18.RGB)
  MaterialFunctionCall(Emissive=Multiply(A=Multiply(A=@n6.Result, B=@n19), B=ScalarParameter() {defaultValue=1, parameterName=All Light})) {materialFunction=&MF_FX_EmissiveAdjustments}
  n20 = Multiply(A=@n10.R, B=@n17)
  n21 = TextureCoordinate()
  n22 = Constant() {r=0.5}
  n23 = Constant() {r=0.5}
  n24 = ScalarParameter() {defaultValue=0.5, parameterName=SphereMask_Hardness}
  n25 = Multiply(
   A=Multiply(A=Multiply(A=@n18.A, B=@n20), B=SphereMask(A=@n21, B=@n22, Radius=@n23, Hardness=@n24))
   B=MaterialFunctionCall("FadeSensitivity (optional)"=ScalarParameter() {parameterName=MeshFade}) {materialFunction=&FadeBasedOnViewAngle}.Result
  )
  n26 = DepthFade(Opacity=@n25, FadeDistance=ScalarParameter() {defaultValue=5, parameterName=DepthFade})
  ShadingPathSwitch(Default=@n26, Deferred=@n26, Forward=@n25, Mobile=@n25)
 ]
}
