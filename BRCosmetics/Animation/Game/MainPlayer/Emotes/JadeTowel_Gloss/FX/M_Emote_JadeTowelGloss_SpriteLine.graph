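# Unlit, two-sided translucent Niagara sprite and mesh material for line effects in the JadeTowel Gloss emote.
# Use_NoiseLine blends a radially panned ring mask with ceiled scrolling-noise lines, whose modulating mask pans via dynamic-parameter speed.
# A sphere mask trims the shape, tinted by particle color and an optional radial two-color gradient.
# Opacity adds view-angle fade and depth fade, skipping depth fade on forward and mobile.
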Material /BRCosmetics/Animation/Game/MainPlayer/Emotes/JadeTowel_Gloss/FX/M_Emote_JadeTowelGloss_SpriteLine
graph {
 properties {
  bAutomaticallySetUsageInEditor=false
  bUsedWithNiagaraMeshParticles=true
  bUsedWithNiagaraSprites=true
  blendMode=BLEND_Translucent
  shadingModel=MSM_Unlit
  twoSided=true
  usedShadingModels=Unlit
 }
 nodes [
  n1 = Multiply(
   A=MaterialFunctionCall() {materialFunction=&VectorToRadialValue}.Radial Coordinates
   B=AppendVector(A=Constant() {r=8}, B=Constant())
  )
  n2 = LinearInterpolate(Alpha=TextureSample!(UVs=Panner(Coordinate=@n1) {speedY=-1.5}, Tex, Level, Apply View MipBias) {mipValueMode=TMVM_MipLevel, texture=/Game/Effects/Fort_Effects/Textures/Tile/Tests/T_Tiling_PanningMask_Softest_Seamless}.R) {constA=0.5}
  n3 = SphereMask(A=TextureCoordinate(), B=Constant() {r=0.5}, Radius=Constant() {r=0.5}, Hardness=Constant())
  n4 = OneMinus(None=@n3)
  n5 = Multiply(A=Power(Base=Multiply(A=@n2) {constB=2}), B=Multiply(A=Power(Base=@n4) {constExponent=1}, B=@n3))
  n6 = TextureCoordinate() {uTiling=0.1, vTiling=3}
  n7 = Constant() {r=5}
  n8 = TextureSample!(UVs=Panner(Coordinate=Multiply(A=@n6, B=@n7)) {speedX=-2}, Tex, Apply View MipBias) {samplerType=SAMPLERTYPE_LinearColor, texture=/Game/TheOutpost/Textures/AnimatedOutpost_Base046_rgb_XVector_a_NormalizedHierPos_UV_2}
  n9 = Ceil(None=LinearInterpolate(Alpha=@n8.R) {constA=-0.2, constB=1.5})
  n10 = Abs(None=Sine(None=Time()))
  n11 = Panner(Coordinate=Multiply(A=TextureCoordinate() {uTiling=0.3, vTiling=2}, B=@n7)) {speedX=1}
  n12 = LinearInterpolate(Alpha=Max(
    A=Subtract(A=@n8.R, B=@n10)
    B=TextureSample(UVs=@n11) {samplerType=SAMPLERTYPE_LinearColor, texture=&AnimatedOutpost_Base046_rgb_XVector_a_NormalizedHierPos_UV_2}.R
   )) {constA=-0.2, constB=1.5}
  n13 = TextureCoordinate() {vTiling=3}
  n14 = AppendVector(
   A=Multiply(A=Constant() {r=1}, B=DynamicParameter() {paramNames [NoiseSpeed_Control, Param2, Param3, Param4]}.NoiseSpeed_Control)
   B=Constant()
  )
  n15 = TextureSample(UVs=MaterialFunctionCall(UVs=Panner(Coordinate=@n13, Speed=@n14) {speedX=1}, "Rotation Angle (0-1)"=Constant() {r=0.25}) {materialFunction=&CustomRotator}.Rotated Values) {texture=&T_Tiling_PanningMask_Softest_Seamless}
  n16 = Multiply(A=Subtract(A=Max(A=@n9), B=Ceil(None=@n12)) {constB=0}, B=Power(Base=Multiply(A=@n15.R)))
  n17 = Multiply(
   A=LinearInterpolate(A=Saturate(None=@n5), B=Saturate(None=@n16), Alpha=ScalarParameter() {parameterName=Use_NoiseLine})
   B=SphereMask(A=TextureCoordinate(), B=Constant() {r=0.5}, Radius=Constant() {r=0.5}, Hardness=Constant())
  )
  n18 = ParticleColor()
  n19 = Multiply(A=@n17, B=@n18.RGB)
  n20 = SphereMask(A=TextureCoordinate(), B=Constant() {r=0.5}, Radius=Constant() {r=0.5}, Hardness=Constant())
  n21 = Multiply(
   A=LinearInterpolate(
    A=VectorParameter() {parameterName=color01, defaultValue=<rgba 0 0.1655 1 1>}.RGB
    B=VectorParameter() {parameterName=color02, defaultValue=<rgba 0 1 0.7462 1>}.RGB
    Alpha=Power(Base=@n20) {constExponent=1.5}
   )
   B=@n19
  )
  n22 = ScalarParameter() {parameterName=Use_SphereColorLerp}
  MaterialFunctionCall(Emissive=Multiply(A=LinearInterpolate(A=@n19, B=@n21, Alpha=@n22), B=ScalarParameter() {defaultValue=1, parameterName=All Light})) {materialFunction=&MF_FX_EmissiveAdjustments}
  n23 = Multiply(
   A=Multiply(A=@n18.A, B=@n17)
   B=MaterialFunctionCall("FadeSensitivity (optional)"=ScalarParameter() {parameterName=MeshFade}) {materialFunction=&FadeBasedOnViewAngle}.Result
  )
  n24 = DepthFade(Opacity=@n23, FadeDistance=ScalarParameter() {defaultValue=5, parameterName=DepthFade})
  ShadingPathSwitch(Default=@n24, Deferred=@n24, Forward=@n23, Mobile=@n23)
 ]
}
