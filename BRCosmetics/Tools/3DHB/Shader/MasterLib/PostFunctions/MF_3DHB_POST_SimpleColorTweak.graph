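# Material function for the 3DHB master shader's post stage that color-grades an emissive input and outputs its replacement.
# A single signed Brightness parameter drives two reciprocal-exponent power curves, one for positive values and one on the inverted color for negative values.
# The result is hue-shifted by an offset plus a gradient from rotated, recentered UVs, then desaturated (negative default boosts saturation) and multiplied by a tint.
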MaterialFunction /BRCosmetics/Tools/3DHB/Shader/MasterLib/PostFunctions/MF_3DHB_POST_SimpleColorTweak
graph {nodes [
  n1 = FunctionInput(Preview=TextureCoordinate()) {bUsePreviewValueAsDefault=true, inputName=UVIn, inputType=FunctionInput_Vector2, previewValue=<1 1 1 1>}
  n2 = Rotator(Coordinate=@n1, Time=ScalarParameter() {defaultValue=15, group=POST_ColorTweak, parameterName=ColorTweak_HueShift_UVRotation})
  n3 = Multiply(
   A=Add(A=Multiply(A=@n2) {constB=2}) {constB=-1}
   B=ScalarParameter() {defaultValue=0.5, group=POST_ColorTweak, parameterName=ColorTweak_HueShift_UVFactor}
  )
  n4 = ScalarParameter() {defaultValue=0.2, group=POST_ColorTweak, parameterName=ColorTweak_Brightness}
  n5 = Add(A=Max(A=@n4) {constB=0})
  n6 = OneMinus(None=Power(Base=FunctionInput() {bUsePreviewValueAsDefault=true, inputName=EmissiveIn, previewValue=<1 1 1 1>}, Exp=Divide(B=@n5)))
  n7 = Min(A=@n4) {constB=0}
  n8 = Divide(B=Add(A=Abs(None=@n7)))
  n9 = MaterialFunctionCall(
   Hue Shift Percentage=Add(
    A=MaterialFunctionCall(Float2=@n3) {materialFunction=&BreakOutFloat2Components}.G
    B=ScalarParameter() {defaultValue=-0.3, group=POST_ColorTweak, parameterName=ColorTweak_HueShift}
   )
   Texture=OneMinus(None=Power(Base=@n6, Exp=@n8))
  ) {materialFunction=&HueShift}
  n10 = ScalarParameter() {defaultValue=-0.5, group=POST_ColorTweak, parameterName=ColorTweak_Desaturation}
  FunctionOutput(None=Multiply(
    A=Desaturation(None=@n9.Result, Fraction=@n10)
    B=VectorParameter() {group=POST_ColorTweak, parameterName=ColorTweak_Tint, defaultValue=<rgba 0.5 0.5 0.5 1>}.RGB
   )) {outputName=ReplaceToEmissive}
 ]}
