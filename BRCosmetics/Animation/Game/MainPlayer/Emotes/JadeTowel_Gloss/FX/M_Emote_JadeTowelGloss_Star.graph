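# Unlit translucent, two-sided Niagara sprite and mesh particle material for star effects in the JadeTowel Gloss emote.
# A star texture plus radial ramp forms the shape, optionally mixed with screen-space dots and tinted by particle color or a scrolling rainbow gradient.
# Dynamic parameters drive ramp, dots, rainbow, noise-modulated vertical offset and pixelated height-lerp dissolve.
# Opacity adds optional scanlines, view-angle fade and depth fade, skipped on mobile and forward paths.
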Material /BRCosmetics/Animation/Game/MainPlayer/Emotes/JadeTowel_Gloss/FX/M_Emote_JadeTowelGloss_Star
graph {
 properties {
  bAutomaticallySetUsageInEditor=false
  bUsedWithNiagaraMeshParticles=true
  bUsedWithNiagaraSprites=true
  blendMode=BLEND_Translucent
  shadingModel=MSM_Unlit
  twoSided=true
  usedShadingModels=Unlit
 }
 nodes [
  n1 = VectorParameter() {parameterName=DissolveTex_Contral, defaultValue=<rgba 0.6 0.6 0 0>}
  AppendVector(A=@n1.B, B=@n1.B)
  n2 = Multiply(A=Add(A=TextureCoordinate(), B=ParticleRandom()), B=ScalarParameter() {defaultValue=1.5, group=Color, parameterName=Color Scale}) {constB=0.9}
  n3 = Constant() {r=0.25}
  n4 = TextureSample(UVs=Panner(Coordinate=MaterialFunctionCall(UVs=@n2, "Rotation Angle (0-1)"=@n3) {materialFunction=&CustomRotator}.Rotated Values) {speedX=-1}) {texture=/Game/Global/Textures/Default/Masks_and_Gradients/T_Grad_Rainbow}
  n5 = ScalarParameter() {group=Color, parameterName=Desaturation}
  n6 = ParticleColor()
  n7 = MaterialFunctionCall(World Position=WorldPosition().XYZ) {materialFunction=&TransformToClipSpace}
  n8 = AppendVector(
   A=Multiply(A=ComponentMask(None=@n7.Buffer UVs for Scene Textures) {r=true}) {constB=100}
   B=Multiply(A=ComponentMask(None=@n7.Buffer UVs for Scene Textures) {g=true}) {constB=50}
  )
  n9 = TextureSampleParameter2D(UVs=Add(A=TextureCoordinate(), B=Constant2Vector() {g=0.015})) {parameterName=BaseTexture, texture=&/T_Emote_JadeTowelGloss_StarRGB}
  n10 = ScalarParameter() {parameterName=Use_StarGlowTexture}
  n11 = Add(
   A=MaterialFunctionCall() {materialFunction=&VectorToRadialValue}.Radial Coordinates
   B=DynamicParameter() {defaultValue=<rgba -2 0 1 1>, paramNames [Mask_Add,Use_AddPoint,WPO,Dissolve]}.Mask_Add
  ) {constB=-1.5}
  n12 = Add(
   A=LinearInterpolate(A=LinearInterpolate(A=@n9.R, B=@n9.R, Alpha=@n10), B=@n9.B, Alpha=ScalarParameter() {parameterName=Use_StarCenterGlowTexture})
   B=Multiply(A=TextureSample(UVs=@n11) {samplerSource=SSM_Clamp_WorldGroupSettings, texture=/Game/Effects/Fort_Effects/Textures/Ramps/T_Power_Source_Sphere_Ramp}.R) {constB=5}
  )
  n13 = Add(
   A=TextureSampleParameter2D(UVs=Multiply(A=@n8) {constB=2}) {parameterName=PointTexture, texture=/Game/Effects/Fort_Effects/Textures/Mask/T_Sphere_A}.R
   B=Multiply(A=Saturate(None=@n12)) {constB=0.2}
  )
  n14 = DynamicParameter() {defaultValue=<rgba -2 0 1 1>, paramNames [Mask_Add,Use_AddPoint,WPO,Dissolve]}
  n15 = LinearInterpolate(A=@n6.RGB, B=Multiply(A=Saturate(None=@n13), B=@n6.RGB), Alpha=@n14.Use_AddPoint)
  n16 = Multiply(A=Multiply(A=Desaturation(None=@n4.RGB, Fraction=@n5), B=ScalarParameter() {defaultValue=8, parameterName=RainbowMul}) {constB=8}, B=@n15)
  n17 = Multiply(
   A=ScalarParameter() {parameterName=Use_RainbowColorTex}
   B=DynamicParameter() {parameterIndex=1, defaultValue=<rgba 0 1 1 1>, paramNames [RainbowColor,Param2,Param3,Param4]}.RainbowColor
  )
  MaterialFunctionCall(Emissive=Multiply(A=LinearInterpolate(A=@n15, B=@n16, Alpha=@n17), B=ScalarParameter() {defaultValue=1, parameterName=Emissive}) {constB=100}) {materialFunction=&MF_FX_EmissiveAdjustments}
  n18 = Constant3Vector() {constant=<rgba 0 0 1 0>}
  n19 = TextureSample(UVs=Panner(Coordinate=TextureCoordinate() {uTiling=0, vTiling=0}) {speedY=0.3}) {samplerType=SAMPLERTYPE_LinearGrayscale, texture=/Engine/Functions/Engine_MaterialFunctions02/ExampleContent/Textures/LowResBlurredNoise}
  Multiply(
   A=Multiply(
    A=TransformPosition(None=@n18) {transformSourceType=TRANSFORMPOSSOURCE_World, transformType=TRANSFORMPOSSOURCE_World}
    B=LinearInterpolate(Alpha=@n19.R) {constA=0.5, constB=0.8}
   )
   B=DynamicParameter() {defaultValue=<rgba -2 0 1 1>, paramNames [Mask_Add,Use_AddPoint,WPO,Dissolve]}.WPO
  )
  n20 = MaterialFunctionCall!(MeshScaleIn=Constant() {r=1}) {materialFunction=/Game/Tools/3DHB/Shader/Base/MF_3DHB_Base}
  NamedRerouteDeclaration(None=ComponentMask(None=@n20.OriginDiffuse) {b=true, g=true, r=true}) {name=Base_Emissive, nodeColor=<rgba 0.5703 1 0 1>}
  NamedRerouteDeclaration(None=ComponentMask(None=@n20.OriginDiffuse) {a=true}) {name=Base_Opacity, nodeColor=<rgba 0.5703 1 0 1>}
  NamedRerouteDeclaration(None=@n20.OriginWPO) {name=Base_WPO, nodeColor=<rgba 0.5703 1 0 1>}
  NamedRerouteDeclaration(None=@n20.OriginNormal) {name=Base_Normal, nodeColor=<rgba 0.5703 1 0 1>}
  n21 = MaterialFunctionCall!(UVIn, EmissiveIn=NamedRerouteUsage()) {materialFunction=/BRCosmetics/Tools/3DHB/Shader/MasterLib/PostFunctions/MF_3DHB_POST_SimpleColorTweak}
  n22 = NamedRerouteUsage()
  n23 = NamedRerouteUsage()
  n24 = MaterialFunctionCall!(
   Time=NamedRerouteUsage()
   "NormalIn(TS)"=@n22
   DirctionIn
   AdditionalDesaturationIn
   AdditionalStrengthIn
   AdditionalDensityIn
   AdditionalFlattenNormalIn
   WPOIn=@n23
  ) {materialFunction=/Game/Tools/3DHB/Shader/MasterLib/PostFunctions/MF_3DHB_POST_RainbowColor}
  NamedRerouteDeclaration(None=Add(
    A=Add(A=@n21.ReplaceToEmissive, B=@n24.BlendToEmissive)
    B=MaterialFunctionCall!("NormalIn(TS)"=@n22) {materialFunction=/BRCosmetics/Tools/3DHB/Shader/MasterLib/PostFunctions/MF_3DHB_POST_RimLight}.AddToEmissive
   )) {name=Post_Emissive, nodeColor=<rgba 0.5703 1 0 1>}
  NamedRerouteDeclaration(None=NamedRerouteUsage()) {name=Post_Opacity, nodeColor=<rgba 0.5703 1 0 1>}
  NamedRerouteDeclaration(None=@n23) {name=Post_WPO, nodeColor=<rgba 0.5703 1 0 1>}
  n25 = Time()
  Panner(Coordinate=Multiply(A=TextureCoordinate(), B=AppendVector(A=@n1.R, B=@n1.R)), Time=@n25) {speedX=0.2, speedY=0.2}
  n26 = Constant() {r=12}
  n27 = Floor(None=Multiply(A=TextureCoordinate() {uTiling=2, vTiling=2}, B=@n26))
  n28 = Time()
  n29 = Constant2Vector()
  n30 = TextureSampleParameter2D(UVs=Add(A=Divide(A=@n27, B=@n26), B=Multiply(A=@n28, B=@n29))) {parameterName=DissolveTex, samplerType=SAMPLERTYPE_LinearColor, texture=/Game/Global/Textures/Default/Noise/T_Wrap_SnowStreak_Noise}
  Saturate(None=Power(Base=Multiply(A=@n30.R) {constB=2}) {constExponent=5})
  n31 = Saturate(None=@n12)
  n32 = Multiply(A=@n6.A, B=Multiply(A=@n31) {constB=0.5})
  n33 = Multiply(A=LinearInterpolate(A=Multiply(A=@n6.A, B=@n31), B=@n32, Alpha=@n14.Use_AddPoint), B=@n32)
  n34 = DynamicParameter() {defaultValue=<rgba -2 0 1 1>, paramNames [Mask_Add,Use_AddPoint,WPO,Dissolve]}
  n35 = LinearInterpolate(Alpha=@n30.R) {constA=-1}
  n36 = LinearInterpolate(
   A=@n33
   B=Multiply(
    A=@n33
    B=MaterialFunctionCall(Transition Phase=@n34.Dissolve, Height Texture=@n35) {materialFunction=&HeightLerp}.Alpha
   )
   Alpha=ScalarParameter() {parameterName=Use_Dissolve}
  )
  n37 = MaterialFunctionCall!(
   Number of Scanlines=ScalarParameter() {defaultValue=180, parameterName=Scanline Amount}
   Scanline Speed
   "Scanlines Thickness (Bias)"=ScalarParameter() {defaultValue=-0.1, parameterName=ScanlineSize}
   Scanlines Dark Value=ScalarParameter() {defaultValue=0.5, parameterName=Scanlines Dark}
   Scanlines Light Value=ScalarParameter() {defaultValue=1, parameterName=Scanlines Light}
   Block Opacity
   UVs
   Time
   Block Repeat
  ) {materialFunction=/Game/Effects/Fort_Effects/Effects/Atlas/Materials/MF_Atlas_Scanlines_01}
  n38 = Multiply(
   A=LinearInterpolate(A=@n36, B=Multiply(A=@n36, B=@n37.Result), Alpha=ScalarParameter() {parameterName=Use_Scanline})
   B=MaterialFunctionCall("FadeSensitivity (optional)"=ScalarParameter() {parameterName=MeshFade}) {materialFunction=&FadeBasedOnViewAngle}.Result
  )
  n39 = DepthFade(Opacity=@n38, FadeDistance=ScalarParameter() {defaultValue=2, parameterName=DepthFade})
  ShadingPathSwitch(Default=@n39, Deferred=@n39, Forward=@n38, Mobile=@n38)
  Sine(None=@n25)
 ]
}
