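# Niagara burst-light effect for the JadeTowel_Gloss character emote, flashing multicolored rays, rings and stage lights.
# Burst_Ray spawns short-lived mesh shafts over a widening area, Burst_Ring expands one mesh ring, and Burst_backboard flashes camera-offset sprites.
# Burst_StageLights places six rotating lights via partition-indexed arrays, with glow color driven by a sine scaled by System.AudioScale.
# Curves drive color and fades, with adjustments for time of day, front end, platform and distance.
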NiagaraSystem /BRCosmetics/Animation/Game/MainPlayer/Emotes/JadeTowel_Gloss/FX/NS_Emote_JadeTowelGloss_Burstlight
dependencies {
 usedDataInterfaces [
  &NiagaraDataInterfaceArrayFloat3
  &NiagaraDataInterfaceCurve
  &NiagaraDataInterfaceColorCurve
  &NiagaraDataInterfaceVectorCurve
  &NiagaraDataInterfacePlatformSet
 ]
 usedDynamicInputs [&ConvertVectorToPosition, &RemapRange]
 usedModules [
  &self:SystemSpawnScript.SystemScriptSource.SystemScriptGraph.NiagaraNodeAssignment_0.SetVariables_#1
  &self:SystemSpawnScript.SystemScriptSource.SystemScriptGraph.NiagaraNodeAssignment_1.SetVariables_#2
  &SystemState
 ]
}
emitters {
 Burst_Ray {
  emitterUpdateScript [
   &EmitterState(Life Cycle Mode=::NewEnumerator1, Loop Behavior=::NewEnumerator1, Loop Duration=0.4, UseLoopDelay=true)
   &SpawnRate(
    SpawnRate=&FloatFromCurve(FloatCurve=DataInterface() {propertyValues {Curve {keys [<at 0 0.6>,<at 1 1>]}}}, CurveIndex=$Emitter.Age, Scale Curve=100)
    Use Spawn Probability=false
   )
   &SpawnBurst_Instantaneous(Spawn Count=5, Use Spawn Probability=false, UseLoopCountLimit=false)
  ]
  particleSpawnScript [
   &InitializeParticle(
    Color=<rgba 0.3015 1 0.798 1>
    Lifetime Max=0.5
    Lifetime Min=0.15
    Lifetime Mode=::NewEnumerator1
    Mesh Scale=&Multiply_VectorByFloat(Vector=&RandomRangeVector(Minimum=<0.2 0.2 1.5>, Maximum=<0.4 0.4 2>), Float=0.1)
    Mesh Scale Mode=::NewEnumerator4
    UsePositionOffset=false
    Write Material Random=true
   )
   &ShapeLocation(
    Handle Radius=10
    Large Radius=&FloatFromCurve(
     FloatCurve=DataInterface() {propertyValues {Curve {keys [
         {arriveTangent=1.8769, interpMode=RCIM_Cubic, leaveTangent=1.8769, tangentMode=RCTM_User, time=0, value=0.2}
         {arriveTangent=0.8, interpMode=RCIM_Cubic, leaveTangent=0.8, time=0.3519, value=0.7183}
         {arriveTangent=0.0876, interpMode=RCIM_Cubic, leaveTangent=0.0876, tangentMode=RCTM_User, time=1, value=1}
        ]}}}
     CurveIndex=$Emitter.Age
     Scale Curve=150
    )
    MaskSpawnGroup=false
    Override Seed=false
    Shape Primitive=::NewEnumerator3
    UseOwnerScale=true
   ) {module=ShapeLocation001}
  ]
  particleUpdateScript [
   &ParticleState()
   &ScaleMeshSize(Scale Factor=&VectorFromCurve(VectorCurve=DataInterface() {propertyValues {
       XCurve {keys [<at 0 1 RCIM_Cubic>,<at 1 1>]}
       YCurve {keys [<at 0 1 RCIM_Cubic>,<at 1 1>]}
       ZCurve {keys [
         {arriveTangent=1.4897, interpMode=RCIM_Cubic, leaveTangent=1.4897, tangentMode=RCTM_User, time=0, value=0.3}
         {arriveTangent=0.2789, tangentMode=RCTM_User, time=1, value=1.5}
        ]}
      }}))
   &DynamicMaterialParameters(
    Index 0 Param 1=1
    Index 0 Param 2=1
    Index 0 Param 3=1
    Index 0 Param 4=1
    Module.Index0Param0Write=true
    Module.Index0Param1Write=true
    Module.Index0Param2Write=true
    Module.Index0Param3Write=true
   )
   &CameraOffset(Camera Offset Amount=15)
   &ScaleColor(
    Scale Alpha=&FloatFromCurve(
     FloatCurve=DataInterface() {propertyValues {Curve {keys [{arriveTangent=0.0239, interpMode=RCIM_Cubic, leaveTangent=0.0239, tangentMode=RCTM_User, time=0, value=1}, <at 1 0>]}}}
     Scale Curve=&Multiply_Float(
      A=&Multiply_Float(A=1, B=&FNTimeOfDayScalar(Night Scale=0.5, Morning Scale=1, Evening Scale=1))
      B=&FN_FrontEndFloatModifier(Gameplay=1, Front End=0.1)
     )
    )
    Scale RGB=&Multiply_Vector(
     A=&VectorFromFloat(Value=8)
     B=&MakeVectorFromLinearColorRGB(Color=&ColorFromCurve(ColorCurve=DataInterface() {propertyValues {
         AlphaCurve {keys [<at 0 1 RCIM_Cubic>]}
         BlueCurve {keys [<at 0.0176 0.7275>,<at 0.1843 0.5265>,<at 0.4216 1>,<at 0.548 1>,<at 0.7273 0.1>,<at 1 0.599 RCIM_Cubic>]}
         GreenCurve {keys [<at 0.0176 1>,<at 0.1843 1>,<at 0.4216 0.5705>,<at 0.548 0>,<at 0.7273 1>,<at 1 1 RCIM_Cubic>]}
         RedCurve {keys [<at 0.0176 0>,<at 0.1843 0>,<at 0.4216 0>,<at 0.548 0.3031>,<at 0.7273 0>,<at 1 0 RCIM_Cubic>]}
        }}))
    )
    ScaleA=true
    ScaleRGB=true
   )
   &FN_Mesh_Attributes(
    Color=&MakeLinearColorFromVectorAndFloat(
     "Vector (RGB)"=&Multiply_VectorByFloat(
      Vector=&MakeVectorFromLinearColorRGB(Color=&ColorFromCurve(ColorCurve=DataInterface() {propertyValues {
          AlphaCurve {keys [<at 0 1 RCIM_Cubic>]}
          BlueCurve {keys [<at 0.0176 0.7275>,<at 0.1843 0.5265>,<at 0.4216 1>,<at 0.548 1>,<at 0.7273 0.1>,<at 1 0.599 RCIM_Cubic>]}
          GreenCurve {keys [<at 0.0176 1>,<at 0.1843 1>,<at 0.4216 0.5705>,<at 0.548 0>,<at 0.7273 1>,<at 1 1 RCIM_Cubic>]}
          RedCurve {keys [<at 0.0176 0>,<at 0.1843 0>,<at 0.4216 0>,<at 0.548 0.3031>,<at 0.7273 0>,<at 1 0 RCIM_Cubic>]}
         }}))
      Float=10
     )
     "Float (Alpha)"=&FloatFromCurve(
      FloatCurve=DataInterface() {propertyValues {Curve {keys [{arriveTangent=0.0703, interpMode=RCIM_Cubic, leaveTangent=0.0703, tangentMode=RCTM_User, time=0, value=1}, <at 1 0>]}}}
      Scale Curve=10
     )
    )
    Create Color Attribute=true
    Create Scale Attribute=true
    Scale=<0.8 0.8 0.1>
    Write to Module Namespace=true
   )
  ]
  rendererClasses [&NiagaraMeshRendererProperties]
  renderers [
   {rendererClass=&NiagaraMeshRendererProperties, rendererIndex=0}
   {rendererClass=&NiagaraMeshRendererProperties, rendererIndex=1}
  ]
 }
 Burst_Ring {
  emitterUpdateScript [
   &EmitterState(Inactive Response=::NewEnumerator1, Life Cycle Mode=::NewEnumerator1, Loop Behavior=::NewEnumerator1, UseLoopDelay=false)
   &SpawnBurst_Instantaneous(Spawn Count=1, Use Spawn Probability=true)
  ]
  particleSpawnScript [
   &InitializeParticle(Mesh Scale Mode=::NewEnumerator3, Mesh Uniform Scale=3, Position Offset=<0 0 10>, UsePositionOffset=true, Write Material Random=true) {module=InitializeParticle001}
   &InitialMeshOrientation(Orientation Axis=<0 0 1>, Orientation Vector=<0 0 1>, Rotation=<0 0 0.25>, Use Orientation Vector=true, Use Rotation Vector=true)
   &InitialMeshOrientation(
    Orientation Axis=<0 0 0>
    Orientation Vector=<0 0 0>
    Rotation=&RandomRangeVector(Minimum=<0 0 0>, Maximum=<0 0 0>)
    Use Orientation Vector=false
    Use Rotation Vector=true
   ) {module=InitialMeshOrientation001}
   /SetVariables() {isSetParameters=true}
   /SetVariables(Particles.Position002:NiagaraPosition=<0 0 20>) {isSetParameters=true}
  ]
  particleUpdateScript [
   &ParticleState()
   &ScaleColor(
    Scale Alpha=&FloatFromCurve(
     FloatCurve=DataInterface() {propertyValues {Curve {keys [
         {arriveTangent=-3.8906, interpMode=RCIM_Cubic, leaveTangent=-3.8906, tangentMode=RCTM_User, time=0, value=1}
         {arriveTangent=-1.0522, interpMode=RCIM_Cubic, leaveTangent=-1.0522, tangentMode=RCTM_User, time=0.2, value=0.4}
         <at 1 0 RCIM_Cubic>
        ]}}}
     Scale Curve=&FN_PickFloatDependOnPlatform(TargetPlatformActive=1, TargetPlatformNoActive=0.7)
    )
    Scale RGB=&Multiply_Vector(
     A=&MakeVectorFromLinearColorRGB(Color=&ColorFromCurve(ColorCurve=DataInterface() {propertyValues {
         AlphaCurve {keys []}
         BlueCurve {keys [<at 0.1142 0.7901 RCIM_Cubic>,<at 0.259 0.7703>,<at 0.6726 1>]}
         GreenCurve {keys [<at 0.1142 0.0197 RCIM_Cubic>,<at 0.259 1>,<at 0.6726 0.4111>]}
         RedCurve {keys [<at 0.1142 1 RCIM_Cubic>,<at 0.259 0>,<at 0.6726 0>]}
        }}))
     B=&VectorFromFloat(Value=&FloatFromCurve(
       FloatCurve=DataInterface() {propertyValues {Curve {keys [{arriveTangent=-1.3603, interpMode=RCIM_Cubic, leaveTangent=-1.3603, tangentMode=RCTM_User, time=0, value=0.8}, <at 1 0 RCIM_Cubic>]}}}
       Scale Curve=0.4
      ))
    )
    ScaleA=true
    ScaleRGB=true
   ) {module=ScaleColor001}
   &ScaleMeshSize(Scale Factor=&VectorFromFloat(Value=&FloatFromCurve(FloatCurve=DataInterface() {propertyValues {Curve {keys [
          {arriveTangent=4.526, interpMode=RCIM_Cubic, leaveTangent=4.526, tangentMode=RCTM_User, time=0, value=0}
          {arriveTangent=2, interpMode=RCIM_Cubic, leaveTangent=2, time=0.3, value=1.25}
          <at 1 2 RCIM_Cubic>
         ]}}})))
   /SetVariables(Particles.GlowColor=&MakeLinearColorFromVectorAndFloat(
     "Vector (RGB)"=&MakeVectorFromLinearColorRGB(Color=&ColorFromCurve(ColorCurve=DataInterface() {propertyValues {
         AlphaCurve {keys []}
         BlueCurve {keys [<at 0.1724 0.5668 RCIM_Cubic>,<at 0.365 1>,<at 0.5023 0>,<at 0.7173 1>,<at 0.8592 1 RCIM_Cubic>]}
         GreenCurve {keys [<at 0.1724 1 RCIM_Cubic>,<at 0.365 0.4385>,<at 0.5023 0.5227>,<at 0.7173 0.4261>,<at 0.8592 0.9951 RCIM_Cubic>]}
         RedCurve {keys [<at 0.1724 0 RCIM_Cubic>,<at 0.365 0>,<at 0.5023 1>,<at 0.7173 0>,<at 0.8592 0 RCIM_Cubic>]}
        }}))
     "Float (Alpha)"=&FloatFromCurve(
      FloatCurve=DataInterface() {propertyValues {Curve {keys [
          {arriveTangent=-1.8164, interpMode=RCIM_Cubic, leaveTangent=-1.8164, tangentMode=RCTM_User, time=0, value=0.4}
          {arriveTangent=-0.0068, interpMode=RCIM_Cubic, leaveTangent=-0.0068, tangentMode=RCTM_User, time=0.1, value=0.8}
          {arriveTangent=-0.11, interpMode=RCIM_Cubic, leaveTangent=-0.11, tangentMode=RCTM_User, time=0.7, value=0}
         ]}}}
      Scale Curve=&FN_FrontEndFloatModifier(Gameplay=1, Front End=0.5)
     )
    )) {isSetParameters=true}
   /SetVariables(Particles.Size002:Vector3f=&Multiply_Vector(
     A=&VectorFromFloat(Value=&FloatFromCurve(
       FloatCurve=DataInterface() {propertyValues {Curve {keys [
           {arriveTangent=6.412, interpMode=RCIM_Cubic, leaveTangent=6.412, tangentMode=RCTM_User, time=0, value=0}
           {arriveTangent=2, interpMode=RCIM_Cubic, leaveTangent=2, time=0.1633, value=0.7527}
           {arriveTangent=0.3953, interpMode=RCIM_Cubic, leaveTangent=0.3953, tangentMode=RCTM_User, time=0.6, value=1.2}
          ]}}}
       Scale Curve=2.2
      ))
     B=&VectorFromFloat(Value=1)
    )) {isSetParameters=true}
   &DynamicMaterialParameters(
    Index 0 Param 1=&FloatFromCurve(FloatCurve=DataInterface() {propertyValues {Curve {keys [{arriveTangent=-0.466, interpMode=RCIM_Cubic, leaveTangent=-0.466, tangentMode=RCTM_User, time=0, value=0.5}, <at 1 0.2>]}}})
    Index 0 Param 2=&FloatFromCurve(FloatCurve=DataInterface() {propertyValues {Curve {keys [{arriveTangent=-0.1866, interpMode=RCIM_Cubic, leaveTangent=-0.1866, tangentMode=RCTM_User, time=0, value=1}, <at 1 0.9>]}}})
    Index 0 Param 3=&FloatFromCurve(FloatCurve=DataInterface() {propertyValues {Curve {keys [<at 0 0>,<at 0.15 1>,<at 1 1.2>]}}})
    Module.Param0WriteEnabled=true
    Module.Param1WriteEnabled=true
    Module.Param2WriteEnabled=true
    Module.Param3WriteEnabled=false
   )
  ]
  rendererClasses [&NiagaraMeshRendererProperties]
  renderers [
   {rendererClass=&NiagaraMeshRendererProperties, rendererIndex=0}
   {rendererClass=&NiagaraMeshRendererProperties, rendererIndex=1}
  ]
 }
 Burst_StageLights {
  emitterUpdateScript [
   &EmitterState(Inactive Response=::NewEnumerator1, Life Cycle Mode=::NewEnumerator1, Loop Behavior=::NewEnumerator1, UseLoopDelay=false)
   &SpawnBurst_Instantaneous(Spawn Count=6, Use Spawn Probability=false, UseLoopCountLimit=false)
  ]
  particleSpawnScript [
   &PartitionParticles(Number Of Partitions=6)
   &InitializeParticle(
    Color Mode=::NewEnumerator0
    Mesh Scale=<0.3 0.3 1>
    Mesh Scale Mode=::NewEnumerator4
    Sprite Size Mode=::NewEnumerator3
    UsePositionOffset=false
    Write Material Random=true
   )
   &ShapeLocation(
    MaskSpawnGroup=false
    Offset=&SelectVectorFromArray(
     Direct Array Index=$Particles.PartitionParticles.Partition
     Vector Selection Array=DataInterface() {propertyValues {FloatData [<90 100 20>, <150 0 20>, <90 -100 20>, <-90 -100 20>, <-150 0 20>, <-90 100 20>]}}
    )
    Offset Mode=::NewEnumerator0
    Override Seed=false
    Sphere Radius=0.1
    UseOwnerScale=true
   )
   &InitialMeshOrientation(
    Mesh Orientation Mode=::NewEnumerator2
    Rotation=&SelectVectorFromArray(
     Direct Array Index=$Particles.PartitionParticles.Partition
     Vector Selection Array=DataInterface() {propertyValues {FloatData [<-0.3 0 -0.15>, <0 0.3 0>, <0.3 0 0.15>, <0.3 0 -0.15>, <0 -0.3 0>, <0 -0.3 -0.2>]}}
    )
    Use Rotation Vector=true
   )
  ]
  particleUpdateScript [
   &ParticleState()
   &ScaleMeshSize(Scale Factor=&MakeVector(
     X=1.2
     Y=1.2
     Z=&FloatFromCurve(FloatCurve=DataInterface() {propertyValues {Curve {keys [
          <at 0 0.625 RCIM_Cubic>
          {arriveTangent=-0.4312, interpMode=RCIM_Cubic, leaveTangent=-0.4312, tangentMode=RCTM_User, time=0.2, value=1.2}
          <at 1 1 RCIM_Cubic>
         ]}}})
    ))
   &UpdateMeshOrientation(
    Rotation Coordinate Space=::NewEnumerator2
    Rotation Rate=&FloatFromCurve(FloatCurve=DataInterface() {propertyValues {Curve {keys [<at 0 1 RCIM_Cubic>,<at 1 0 RCIM_Cubic>]}}}, Scale Curve=1.4)
    Rotation Vector=&SelectVectorFromArray(
     Direct Array Index=$Particles.PartitionParticles.Partition
     Vector Selection Array=DataInterface() {propertyValues {FloatData [<-0.4 0.4 0>, <0 0.55 0>, <0.4 0.4 0>, <0.4 -0.4 0>, <0 -0.55 0>, <-0.4 -0.4 0>]}}
    )
   )
   &DynamicMaterialParameters(
    Index 0 Param 1=0.15
    Module.Index0Param0Write=true
    Module.Index0Param1Write=true
    Module.Index0Param2Write=true
    Module.Index0Param3Write=true
   )
   /SetVariables(
    Particles.RingPosition:NiagaraPosition=&AddVectorToPosition(Position=$Particles.Position, Vector=<0 0 3>)
    Particles.RingScale:Vector3f=&VectorFromFloat(Value=0.25)
   ) {isSetParameters=true}
   /SetVariables(Particles.DiskScale:Vector3f=&Multiply_VectorByFloat(Vector=<1 1 0.5>, Float=1)) {isSetParameters=true}
   /SetVariables(
    Particles.GlowColor=&MakeLinearColorFromVectorAndFloat(
     "Vector (RGB)"=&Multiply_Vector(
      A=&MakeVectorFromLinearColorRGB(Color=&ColorFromCurve(
        ColorCurve=DataInterface() {propertyValues {
          AlphaCurve {keys []}
          BlueCurve {keys [<at 0.1172 0.807 RCIM_Cubic>,<at 0.259 0.8879>,<at 0.5248 0.019>,<at 0.6546 0.8879>,<at 0.8278 0.807 RCIM_Cubic>]}
          GreenCurve {keys [<at 0.1172 0.8879 RCIM_Cubic>,<at 0.259 0.3043>,<at 0.5248 1>,<at 0.6546 0>,<at 0.8278 0.8879 RCIM_Cubic>]}
          RedCurve {keys [<at 0.1172 0.0052 RCIM_Cubic>,<at 0.259 0>,<at 0.5248 0.1838>,<at 0.6546 0.4788>,<at 0.8278 0.0052 RCIM_Cubic>]}
         }}
        CurveIndex=&Multiply_Float(A=&Sine(Normalized Angle=$Emitter.Age, Period=2, Scale=1, Bias=0), B=$System.AudioScale)
       ))
      B=&VectorFromFloat(Value=1)
     )
     "Float (Alpha)"=0.3
    )
    Particles.GlowScale:Vector2f=&Vector2DFromFloat(Value=65)
   ) {isSetParameters=true}
   &ScaleColor(
    Scale Alpha=&FloatFromCurve(FloatCurve=DataInterface() {propertyValues {Curve {keys [<at 0.5 1 RCIM_Cubic>,<at 1 0 RCIM_Cubic>]}}})
    Scale RGB=&Multiply_Vector(
     A=&MakeVectorFromLinearColorRGB(Color=&ColorFromCurve(ColorCurve=DataInterface() {propertyValues {
         AlphaCurve {keys []}
         BlueCurve {keys [<at 0.1172 0.807 RCIM_Cubic>,<at 0.259 0.8879>,<at 0.5248 0.019>,<at 0.6546 0.8879>,<at 0.8278 0.807 RCIM_Cubic>]}
         GreenCurve {keys [<at 0.1172 0.8879 RCIM_Cubic>,<at 0.259 0.3043>,<at 0.5248 1>,<at 0.6546 0>,<at 0.8278 0.8879 RCIM_Cubic>]}
         RedCurve {keys [<at 0.1172 0.0052 RCIM_Cubic>,<at 0.259 0>,<at 0.5248 0.1838>,<at 0.6546 0.4788>,<at 0.8278 0.0052 RCIM_Cubic>]}
        }}))
     B=&VectorFromFloat(Value=&Multiply_Float(
       A=&Multiply_Float(A=1.2, B=&FN_FrontEndFloatModifier(Gameplay=1, Front End=0.3))
       B=&FNTimeOfDayScalar(Night Scale=0.6, Morning Scale=1, Evening Scale=1)
      ))
    )
    ScaleA=true
    ScaleRGB=true
   ) {module=ScaleColor001}
  ]
  rendererClasses [&NiagaraMeshRendererProperties, &NiagaraSpriteRendererProperties]
  renderers [
   {rendererClass=&NiagaraMeshRendererProperties, rendererIndex=0}
   {rendererClass=&NiagaraMeshRendererProperties, rendererIndex=1}
   {rendererClass=&NiagaraMeshRendererProperties, rendererIndex=2}
   {rendererClass=&NiagaraSpriteRendererProperties, rendererIndex=3}
  ]
 }
 Burst_backboard {
  emitterUpdateScript [
   &EmitterState(Inactive Response=::NewEnumerator1, Life Cycle Mode=::NewEnumerator1, Loop Behavior=::NewEnumerator1, UseLoopDelay=false)
   &SpawnBurst_Instantaneous(Use Spawn Probability=false, UseLoopCountLimit=false)
  ]
  particleSpawnScript [
   &InitializeParticle(
    Color=<rgba 0.2622 1 0.8395 1>
    Mesh Scale Mode=::NewEnumerator3
    Position Offset=<0 0 125>
    Sprite Size=<400 300>
    Sprite Size Mode=::NewEnumerator4
    UsePositionOffset=true
    Write Material Random=true
   )
   &SpriteFacingAndAlignment(Sprite Alignment=<1 0 0>, Sprite Facing=<0 1 0>, WriteAlignment=true, WriteFacing=true) {enabled=false}
   &CameraOffset(Camera Offset Amount=-50)
   /SetVariables(Particles.CameraOffset002=80) {isSetParameters=true}
  ]
  particleUpdateScript [
   &ParticleState()
   &ScaleMeshSize(Scale Factor=&VectorFromFloat(Value=&FloatFromCurve(
      FloatCurve=DataInterface() {propertyValues {Curve {keys [{arriveTangent=1.0952, interpMode=RCIM_Cubic, leaveTangent=1.0952, tangentMode=RCTM_User, time=0, value=0}, <at 0.3022 0.9871 RCIM_Cubic>]}}}
      Scale Curve=0.2
     )))
   &ScaleSpriteSize(
    Uniform Curve Scale=&FN_LOD_Distance_Remap(Start Distance=0, End Distance=1200, Minimum Output=0.7, Maximum Output=1, Clamp Output=true)
    Uniform Curve Sprite Scale=DataInterface() {propertyValues {Curve {keys [
        {arriveTangent=3.2619, interpMode=RCIM_Cubic, leaveTangent=3.2619, tangentMode=RCTM_User, time=0, value=0}
        {arriveTangent=1, interpMode=RCIM_Cubic, leaveTangent=1, time=0.2591, value=0.7378}
        <at 1 1 RCIM_Cubic>
       ]}}}
   )
   &ScaleColor(
    Scale Alpha=&FloatFromCurve(FloatCurve=DataInterface() {propertyValues {Curve {keys [
         <at 0 0 RCIM_Cubic>
         {arriveTangent=-0.0232, interpMode=RCIM_Cubic, leaveTangent=-0.0232, tangentMode=RCTM_User, time=0.1, value=1.4}
         {arriveTangent=-1.9949, interpMode=RCIM_Cubic, leaveTangent=-1.9949, tangentMode=RCTM_User, time=0.3, value=0.6}
         <at 1 0 RCIM_Cubic>
        ]}}})
    Scale RGB=&MakeVectorFromLinearColorRGB(Color=<rgba 0.2783 4 2.6551 1>)
    ScaleA=true
    ScaleRGB=true
   )
   /SetVariables(Particles.Color003:LinearColor=&MakeLinearColorFromVectorAndFloat(
     "Vector (RGB)"=&MakeVectorFromLinearColorRGB(Color=<rgba 0.1146 1 0.68 1>)
     "Float (Alpha)"=&FloatFromCurve(
      FloatCurve=DataInterface() {propertyValues {Curve {keys [<at 0 0>,<at 0.08 1>,<at 0.1576 0.4381>,<at 0.3 0>]}}}
      Scale Curve=&Multiply_Float(
       A=1.2
       B=&Multiply_Float(A=&FNTimeOfDayScalar(Night Scale=0.8, Morning Scale=1, Evening Scale=1), B=1)
      )
     )
    )) {isSetParameters=true}
   /SetVariables(Particles.Size003:Vector2f=&Multiply_Vector2D(
     A=<300 400>
     B=&Vector2DFromFloat(Value=&FloatFromCurve(FloatCurve=DataInterface() {propertyValues {Curve {keys [<at 0 0.4>,<at 0.3 1>]}}}, Scale Curve=2))
    )) {isSetParameters=true}
  ]
  rendererClasses [&NiagaraSpriteRendererProperties]
  renderers [
   {rendererClass=&NiagaraSpriteRendererProperties}
   {rendererClass=&NiagaraSpriteRendererProperties, rendererIndex=1}
  ]
 }
}
summary {
 systemName=NS_Emote_JadeTowelGloss_Burstlight
 emitters [
  {
   emitterName=Burst_StageLights
   rendererClasses [&NiagaraMeshRendererProperties, &NiagaraSpriteRendererProperties]
  }
  {emitterName=Burst_Ray, rendererClasses [&NiagaraMeshRendererProperties]}
  {emitterName=Burst_Ring, rendererClasses [&NiagaraMeshRendererProperties]}
  {emitterName=Burst_backboard, rendererClasses [&NiagaraSpriteRendererProperties]}
 ]
 userVariables [
  User.bIsFrontEnd {type=NiagaraBool}
  User.bIsFrontEndPreview {type=NiagaraBool}
 ]
}
