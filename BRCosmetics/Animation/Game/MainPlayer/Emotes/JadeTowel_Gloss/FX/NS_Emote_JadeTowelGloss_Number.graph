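# Niagara effect for the JadeTowel Gloss emote that pops up a glowing number mesh, chosen by User.MeshInt, with a burst of accents.
# Flare, ring, sparkle sprites and tumbling mesh bits burst at spawn, followed by delayed color-cycling glitch sprites and meshes.
# Later repeat bursts are disabled.
# Front-end, platform and time-of-day modifiers scale parts of the glow and brightness, while the mesh bits' material reads event audio amplitude.
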NiagaraSystem /BRCosmetics/Animation/Game/MainPlayer/Emotes/JadeTowel_Gloss/FX/NS_Emote_JadeTowelGloss_Number
dependencies {
 usedDataInterfaces [
  &NiagaraDataInterfaceMeshRendererInfo
  &NiagaraDataInterfaceCurve
  &NiagaraDataInterfaceColorCurve
  &NiagaraDataInterfacePlatformSet
  &NiagaraDataInterfaceArrayFloat
  &NiagaraDataInterfaceArrayColor
  &NiagaraDataInterfaceVector2DCurve
 ]
 usedDynamicInputs [&SelectFloatFromArray, &FN_FrontEndVector3Modifier]
 usedModules [&SystemState]
}
emitters {
 Flare {
  emitterUpdateScript [
   &EmitterState(Inactive Response=::NewEnumerator1, Life Cycle Mode=::NewEnumerator1, Loop Behavior=::NewEnumerator1, UseLoopDelay=false)
   &SpawnBurst_Instantaneous(Spawn Count=1, Use Spawn Probability=true)
   &SpawnBurst_Instantaneous(Spawn Count=1, Spawn Time=1.1, Use Spawn Probability=true) {enabled=false, module=SpawnBurst_Instantaneous001}
   &SpawnBurst_Instantaneous(Spawn Count=1, Spawn Time=2.4, Use Spawn Probability=true) {enabled=false, module=SpawnBurst_Instantaneous002}
  ]
  particleSpawnScript [
   &InitializeParticle(Lifetime=0.6, Mesh Scale Mode=::NewEnumerator3, Mesh Uniform Scale=0.4, UsePositionOffset=false, Write Material Random=true) {module=InitializeParticle001}
   &InitialMeshOrientation(Orientation Axis=<0 0 1>, Orientation Vector=<0 0 1>, Rotation=<0 0.25 0>, Use Orientation Vector=true, Use Rotation Vector=true)
   &InitialMeshOrientation(
    Orientation Axis=<0 0 0>
    Orientation Vector=<0 0 0>
    Rotation=&RandomRangeVector(Minimum=<0 0 0>, Maximum=<1 0 0>)
    Use Orientation Vector=false
    Use Rotation Vector=true
   ) {module=InitialMeshOrientation001}
  ]
  particleUpdateScript [
   &ParticleState()
   &ScaleColor(
    Scale Alpha=&FloatFromCurve(
     FloatCurve=DataInterface() {propertyValues {Curve {keys [
         {arriveTangent=-3.8906, interpMode=RCIM_Cubic, leaveTangent=-3.8906, tangentMode=RCTM_User, time=0, value=1}
         {arriveTangent=-1.0522, interpMode=RCIM_Cubic, leaveTangent=-1.0522, tangentMode=RCTM_User, time=0.2, value=0.4}
         <at 1 0 RCIM_Cubic>
        ]}}}
     Scale Curve=&Multiply_Float(
      A=&FN_FrontEndFloatModifier(Gameplay=1, Front End=0.5)
      B=&Multiply_Float(
       A=&FNTimeOfDayScalar(Night Scale=0.5, Morning Scale=1, Evening Scale=1)
       B=&FN_PickFloatDependOnPlatform(TargetPlatformActive=1, TargetPlatformNoActive=0.5)
      )
     )
    )
    Scale RGB=&Multiply_Vector(
     A=&MakeVectorFromLinearColorRGB(Color=&ColorFromCurve(ColorCurve=DataInterface() {propertyValues {
         AlphaCurve {keys []}
         BlueCurve {keys [<at 0.1142 0.7901 RCIM_Cubic>,<at 0.259 0.6877>,<at 0.6726 1>]}
         GreenCurve {keys [<at 0.1142 0.0197 RCIM_Cubic>,<at 0.259 2>,<at 0.6726 0.6669>]}
         RedCurve {keys [<at 0.1142 1 RCIM_Cubic>,<at 0.259 0>,<at 0.6726 0>]}
        }}))
     B=&VectorFromFloat(Value=&FloatFromCurve(
       FloatCurve=DataInterface() {propertyValues {Curve {keys [{arriveTangent=-1.3603, interpMode=RCIM_Cubic, leaveTangent=-1.3603, tangentMode=RCTM_User, time=0, value=0.8}, <at 1 0 RCIM_Cubic>]}}}
       Scale Curve=0.4
      ))
    )
    ScaleA=true
    ScaleRGB=true
   ) {module=ScaleColor001}
   /SetVariables(Particles.GlowColor=&MakeLinearColorFromVectorAndFloat(
     "Vector (RGB)"=&MakeVectorFromLinearColorRGB(Color=&ColorFromCurve(ColorCurve=DataInterface() {propertyValues {
         AlphaCurve {keys []}
         BlueCurve {keys [
           <at 0.0455 0.5>
           {arriveTangent=-0.2054, interpMode=RCIM_Cubic, leaveTangent=-0.2054, time=0.1127, value=0.4715}
           <at 0.1844 0.4715>
           <at 0.2426 1>
           <at 0.365 0.5>
           <at 0.4456 1>
           <at 0.5412 1>
          ]}
         GreenCurve {keys [
           <at 0.0455 0>
           {arriveTangent=7.2026, interpMode=RCIM_Cubic, leaveTangent=7.2026, time=0.1127, value=1}
           <at 0.1844 1>
           <at 0.2426 0.0477>
           <at 0.365 0>
           <at 0.4456 0.25>
           <at 0.5412 0.8688>
          ]}
         RedCurve {keys [
           <at 0.0455 0.1299>
           {arriveTangent=-0.9357, interpMode=RCIM_Cubic, leaveTangent=-0.9357, time=0.1127, value=0}
           <at 0.1844 0>
           <at 0.2426 0>
           <at 0.365 0.4085>
           <at 0.4456 0>
           <at 0.5412 0>
          ]}
        }}))
     "Float (Alpha)"=&FloatFromCurve(
      FloatCurve=DataInterface() {propertyValues {Curve {keys [
          <at 0 0 RCIM_Cubic>
          {arriveTangent=6.6667, interpMode=RCIM_Cubic, leaveTangent=6.6667, time=0, value=0}
          {arriveTangent=-0.8094, interpMode=RCIM_Cubic, leaveTangent=-0.8094, tangentMode=RCTM_User, time=0.15, value=1}
          {arriveTangent=-0.0042, leaveTangent=-1.1765, tangentMode=RCTM_User, time=0.6, value=0}
          <at 1 0 RCIM_Cubic>
         ]}}}
      Scale Curve=&Multiply_Float(
       A=&Multiply_Float(A=0.3, B=&FN_FrontEndFloatModifier(Gameplay=1, Front End=0.5))
       B=&Multiply_Float(
        A=&FNTimeOfDayScalar(Night Scale=0.6, Morning Scale=1, Evening Scale=1)
        B=&FN_PickFloatDependOnPlatform(TargetPlatform=DataInterface() {propertyValues {Platforms {qualityLevelMask=-3}}}, TargetPlatformActive=1, TargetPlatformNoActive=0.6)
       )
      )
     )
    )) {isSetParameters=true}
   /SetVariables(Particles.Scale002:Vector3f=&Multiply_Vector(
     A=<1 1 1>
     B=&VectorFromFloat(Value=&FloatFromCurve(
       FloatCurve=DataInterface() {propertyValues {Curve {keys [
           {arriveTangent=1.1767, interpMode=RCIM_Cubic, leaveTangent=1.1767, tangentMode=RCTM_User, time=0, value=0.4}
           {arriveTangent=0.4177, interpMode=RCIM_Cubic, leaveTangent=0.4177, tangentMode=RCTM_User, time=0.2603, value=0.76}
           {arriveTangent=0.1002, interpMode=RCIM_Cubic, leaveTangent=0.1002, tangentMode=RCTM_User, time=1, value=1}
          ]}}}
       Scale Curve=0.8
      ))
    )) {isSetParameters=true}
   &ScaleMeshSize(Scale Factor=&VectorFromFloat(Value=&FloatFromCurve(FloatCurve=DataInterface() {propertyValues {Curve {keys [<at 0 0>,<at 0.2 1.375>,<at 1 2>]}}})))
   &DynamicMaterialParameters(
    Index 0 Param 1=&FloatFromCurve(FloatCurve=DataInterface() {propertyValues {Curve {keys [{arriveTangent=-0.466, interpMode=RCIM_Cubic, leaveTangent=-0.466, tangentMode=RCTM_User, time=0, value=0.5}, <at 1 0.2>]}}})
    Index 0 Param 2=&FloatFromCurve(FloatCurve=DataInterface() {propertyValues {Curve {keys [{arriveTangent=-0.1866, interpMode=RCIM_Cubic, leaveTangent=-0.1866, tangentMode=RCTM_User, time=0, value=1}, <at 1 0.9>]}}})
    Index 0 Param 3=&FloatFromCurve(FloatCurve=DataInterface() {propertyValues {Curve {keys [<at 0 0>,<at 0.15 1>,<at 1 1.2>]}}})
    Module.Param0WriteEnabled=true
    Module.Param1WriteEnabled=true
    Module.Param2WriteEnabled=true
    Module.Param3WriteEnabled=false
   )
   /SetVariables(Particles.Dyn002:Vector4f=&MakeVector4(
     X=&FloatFromCurve(
      FloatCurve=DataInterface() {propertyValues {Curve {keys [{arriveTangent=-0.6916, interpMode=RCIM_Cubic, leaveTangent=-0.6916, tangentMode=RCTM_User, time=0, value=0.5}, <at 1 0.1>]}}}
      Scale Curve=0.6
     )
     Y=&FloatFromCurve(
      FloatCurve=DataInterface() {propertyValues {Curve {keys [{arriveTangent=-0.1866, interpMode=RCIM_Cubic, leaveTangent=-0.1866, tangentMode=RCTM_User, time=0, value=1}, <at 1 0.9>]}}}
      Scale Curve=0.4
     )
     Z=&FloatFromCurve(FloatCurve=DataInterface() {propertyValues {Curve {keys [<at 0 0>,<at 0.15 1>,<at 1 1.2>]}}})
     W=0
    )) {isSetParameters=true}
  ]
  rendererClasses [&NiagaraMeshRendererProperties]
  renderers [
   {rendererClass=&NiagaraMeshRendererProperties, rendererIndex=0}
   {rendererClass=&NiagaraMeshRendererProperties, rendererIndex=1}
  ]
 }
 Glitch_sprite {
  emitterUpdateScript [
   &EmitterState(Inactive Response=::NewEnumerator1, Life Cycle Mode=::NewEnumerator1, Loop Behavior=::NewEnumerator1, Loop Duration=8, UseLoopDelay=false)
   &SpawnBurst_Instantaneous(Spawn Time=0.8, Use Spawn Probability=false, UseLoopCountLimit=false)
   &SpawnBurst_Instantaneous(Spawn Time=1.9, Use Spawn Probability=false, UseLoopCountLimit=false) {enabled=false, module=SpawnBurst_Instantaneous001}
   &SpawnBurst_Instantaneous(Spawn Time=3.2, Use Spawn Probability=false, UseLoopCountLimit=false) {enabled=false, module=SpawnBurst_Instantaneous002}
  ]
  particleSpawnScript [&InitializeParticle(
    Color=<rgba 5 1.3021 1.3842 2>
    Lifetime=0.3
    Sprite Size=<30 20>
    Sprite Size Mode=::NewEnumerator4
    UsePositionOffset=true
    Write Material Random=true
   )]
  particleUpdateScript [
   &ParticleState()
   &ScaleSpriteSize(
    Non-Uniform Curve Scale=&Vector2DFromFloat(Value=10)
    Non-Uniform Curve Sprite Scale=DataInterface() {propertyValues {
      XCurve {keys [{tangentMode=RCTM_User, time=0, value=0}, {tangentMode=RCTM_User, time=1, value=1}]}
      YCurve {keys [
        {arriveTangent=-1.17, interpMode=RCIM_Cubic, leaveTangent=-1.17, tangentMode=RCTM_User, time=0, value=1}
        {arriveTangent=-0.2416, interpMode=RCIM_Cubic, leaveTangent=-0.2416, tangentMode=RCTM_User, time=1, value=0}
       ]}
     }}
    Scale Sprite Size Mode=::NewEnumerator3
   )
   &ScaleColor(
    Scale Alpha=&Multiply_Float(
     A=&FloatFromCurve(FloatCurve=DataInterface() {propertyValues {Curve {keys [
          {arriveTangent=-2.5249, interpMode=RCIM_Cubic, leaveTangent=-2.5249, tangentMode=RCTM_User, time=0, value=1}
          {arriveTangent=-0.0405, interpMode=RCIM_Cubic, leaveTangent=-0.0405, tangentMode=RCTM_User, time=1, value=0}
         ]}}})
     B=1
    )
    Scale RGB=&Multiply_VectorByFloat(
     Vector=&MakeVectorFromLinearColorRGB(Color=&ColorFromCurve(ColorCurve=DataInterface() {propertyValues {
         AlphaCurve {keys []}
         BlueCurve {keys [
           <at 0 5 RCIM_Cubic>
           <at 0.1326 0.51>
           <at 0.2869 1.4899>
           <at 0.4337 1.0256>
           <at 0.6236 5>
           {arriveTangent=-13.9412, interpMode=RCIM_Cubic, leaveTangent=-13.9412, time=0.7107, value=0}
           <at 0.8097 2.4056>
          ]}
         GreenCurve {keys [
           <at 0 2.5535 RCIM_Cubic>
           <at 0.1326 1>
           <at 0.2869 5>
           <at 0.4337 2>
           <at 0.6236 4.8554>
           {arriveTangent=0.7772, interpMode=RCIM_Cubic, leaveTangent=0.7772, time=0.7107, value=0}
           <at 0.8097 5>
          ]}
         RedCurve {keys [<at 0 0 RCIM_Cubic>,<at 0.1326 0>,<at 0.2869 0>,<at 0.4337 0>,<at 0.6236 0>,<at 0.7107 1 RCIM_Cubic>,<at 0.8097 0>]}
        }}))
     Float=&FloatFromCurve(
      FloatCurve=DataInterface() {propertyValues {Curve {keys [{arriveTangent=-3.0627, interpMode=RCIM_Cubic, leaveTangent=-3.0627, tangentMode=RCTM_User, time=0, value=2}, <at 1 0>]}}}
      Scale Curve=&Multiply_Float(
       A=&Multiply_Float(A=2, B=&FN_FrontEndFloatModifier(Gameplay=1, Front End=0.25))
       B=&Multiply_Float(
        A=&FNTimeOfDayScalar(Night Scale=0.3, Morning Scale=1, Evening Scale=1)
        B=&FN_PickFloatDependOnPlatform(TargetPlatformActive=1, TargetPlatformNoActive=0.5)
       )
      )
     )
    )
    ScaleA=true
    ScaleRGB=true
   )
  ]
  rendererClasses [&NiagaraSpriteRendererProperties]
  renderers [{rendererClass=&NiagaraSpriteRendererProperties}]
 }
 Glitch_sprite002 {
  emitterUpdateScript [
   &EmitterState(Inactive Response=::NewEnumerator1, Life Cycle Mode=::NewEnumerator1, Loop Behavior=::NewEnumerator1, Loop Duration=8, UseLoopDelay=false)
   &SpawnBurst_Instantaneous(Spawn Count=4, Spawn Time=0.8, Use Spawn Probability=false, UseLoopCountLimit=false)
   &SpawnBurst_Instantaneous(Spawn Count=4, Spawn Time=1.9, Use Spawn Probability=false, UseLoopCountLimit=false) {enabled=false, module=SpawnBurst_Instantaneous001}
   &SpawnBurst_Instantaneous(Spawn Count=4, Spawn Time=3.2, Use Spawn Probability=false, UseLoopCountLimit=false) {enabled=false, module=SpawnBurst_Instantaneous002}
  ]
  particleSpawnScript [
   &InitializeParticle(
    Color=<rgba 0 5 3.6342 2>
    Lifetime Max=0.35
    Lifetime Min=0.15
    Lifetime Mode=::NewEnumerator1
    Mesh Scale Max=&MakeCustomVectorFromBool(Bool=&RandomBool(), True Vector=<1.5 2 1>, False Vector=<2 3 5>)
    Mesh Scale Min=<1 1 2>
    Mesh Scale Mode=::NewEnumerator2
    UsePositionOffset=true
    Write Material Random=true
   )
   &ShapeLocation(
    MaskSpawnGroup=false
    Offset Mode=::NewEnumerator0
    Override Seed=false
    Shape Origin=&VectorFromFloat(Value=0)
    Sphere Radius=6
    UseOwnerScale=true
   )
  ]
  particleUpdateScript [
   &ParticleState()
   &ScaleMeshSize(Scale Factor=&Multiply_VectorByFloat(
     Vector=<1 1 1>
     Float=&Multiply_Float(
      A=0.15
      B=&FloatFromCurve(FloatCurve=DataInterface() {propertyValues {Curve {keys [
           {arriveTangent=4.1785, time=0, value=1}
           <at 0.2 1.3>
           {arriveTangent=-0.0405, interpMode=RCIM_Cubic, leaveTangent=-0.0405, tangentMode=RCTM_User, time=1, value=1.5}
          ]}}})
     )
    ))
   &DynamicMaterialParameters(Module.Param0WriteEnabled=true, Module.Param1WriteEnabled=true, Module.Param2WriteEnabled=true, Module.Param3WriteEnabled=true)
   &ScaleColor(
    Scale Alpha=&Multiply_Float(
     A=&FloatFromCurve(
      FloatCurve=DataInterface() {propertyValues {Curve {keys [
          {arriveTangent=9.0725, interpMode=RCIM_Cubic, leaveTangent=9.0725, tangentMode=RCTM_User, time=0, value=0}
          {arriveTangent=-2.5249, interpMode=RCIM_Cubic, leaveTangent=-2.5249, tangentMode=RCTM_User, time=0.2, value=0.9847}
          {arriveTangent=-0.0405, interpMode=RCIM_Cubic, leaveTangent=-0.0405, tangentMode=RCTM_User, time=1, value=0}
         ]}}}
      Scale Curve=&RandomRangeFloat(Minimum=0.2, Maximum=1)
     )
     B=1
    )
    Scale RGB=&Multiply_VectorByFloat(
     Vector=&MakeVectorFromLinearColorRGB(Color=&ColorFromCurve(ColorCurve=DataInterface() {propertyValues {
         AlphaCurve {keys []}
         BlueCurve {keys [
           <at 0 5 RCIM_Cubic>
           <at 0.1326 0.3163>
           <at 0.2869 1.4885>
           <at 0.4134 5>
           <at 0.5108 2.7002>
           <at 0.6236 2.6071>
           {arriveTangent=-1.0828, interpMode=RCIM_Cubic, leaveTangent=-1.0828, time=0.7107, value=0}
           <at 0.8097 2.4056>
          ]}
         GreenCurve {keys [<at 0 2.5535 RCIM_Cubic>,<at 0.1326 1>,<at 0.2869 5>,<at 0.4134 0>,<at 0.5108 5>,<at 0.6236 5>,<at 0.7107 0 RCIM_Cubic>,<at 0.8097 5>]}
         RedCurve {keys [<at 0 0 RCIM_Cubic>,<at 0.1326 0>,<at 0.2869 0>,<at 0.4134 0.4394>,<at 0.5108 0>,<at 0.6236 0>,<at 0.7107 1 RCIM_Cubic>,<at 0.8097 0>]}
        }}))
     Float=&FloatFromCurve(
      FloatCurve=DataInterface() {propertyValues {Curve {keys [{arriveTangent=-2.0684, interpMode=RCIM_Cubic, leaveTangent=-2.0684, tangentMode=RCTM_User, time=0, value=2}, <at 1 0>]}}}
      Scale Curve=2
     )
    )
    ScaleA=true
    ScaleRGB=true
   )
  ]
  rendererClasses [&NiagaraMeshRendererProperties]
  renderers [{rendererClass=&NiagaraMeshRendererProperties, rendererIndex=0}]
 }
 Minimal004 {
  emitterUpdateScript [
   &EmitterState(Inactive Response=::NewEnumerator1, Life Cycle Mode=::NewEnumerator1, Loop Behavior=::NewEnumerator1, UseLoopDelay=false)
   &SpawnBurst_Instantaneous(Spawn Count=15, Use Spawn Probability=false, UseLoopCountLimit=false)
   &SpawnBurst_Instantaneous(Spawn Count=15, Spawn Time=1.1, Use Spawn Probability=false, UseLoopCountLimit=false) {enabled=false, module=SpawnBurst_Instantaneous001}
   &SpawnBurst_Instantaneous(Spawn Count=15, Spawn Time=2.4, Use Spawn Probability=false, UseLoopCountLimit=false) {enabled=false, module=SpawnBurst_Instantaneous002}
  ]
  particleSpawnScript [
   &InitializeParticle(
    Color=&SelectLinearColorFromArray(
     Array Sampling Mode=::NewEnumerator1
     Direct Array Index=&RandomRangeInt(Minimum=0, Maximum=2)
     Color Selection Array=DataInterface() {propertyValues {ColorData [<rgba 0.1394 0.2649 2 1>, <rgba 0 1.6981 2 1>, <rgba 2 0 1.339 1>]}}
     Random Seed=?
    )
    Lifetime Max=1.2
    Lifetime Min=0.2
    Lifetime Mode=::NewEnumerator1
    Mass Max=1.5
    Mass Mode=::NewEnumerator3
    Mesh Scale Mode=::NewEnumerator1
    Mesh Uniform Scale Max=0.5
    Mesh Uniform Scale Min=0.1
    Sprite Rotation Mode=::NewEnumerator2
    Sprite Size Mode=::NewEnumerator3
    UsePositionOffset=false
    Write Material Random=true
   )
   &ShapeLocation(MaskSpawnGroup=false, Override Seed=false, Sphere Radius=5, Sphere Surface Distribution=1, UseOwnerScale=true)
   &InitialMeshOrientation(Use Rotation Vector=false)
   &MeshRotationForce(Rotation=&RandomRangeVector(Minimum=<1 1 1>, Maximum=<3 3 3>))
   &ApplyInitialForces()
   &AddVelocity(Module.Use Falloff Distance=false, Override Seed=false, Velocity Mode=::NewEnumerator1, Velocity Speed=100)
  ]
  particleUpdateScript [
   &ParticleState()
   &ScaleMeshSize(Scale Factor=&VectorFromFloat(Value=&FloatFromCurve(FloatCurve=DataInterface() {propertyValues {Curve {keys [
          {interpMode=RCIM_Cubic, tangentMode=RCTM_User, time=0, value=0}
          {interpMode=RCIM_Cubic, tangentMode=RCTM_User, time=0.2, value=1}
          {arriveTangent=-0.0093, interpMode=RCIM_Cubic, leaveTangent=-0.0093, tangentMode=RCTM_User, time=0.5, value=1}
          {tangentMode=RCTM_User, time=1, value=0}
         ]}}})))
   &DynamicMaterialParameters(
    Index 0 Param 1=&Add_Float(
     A=20
     B=&Lerp_Float(
      A=&Multiply_Float(A=$NPC.FortniteNPC.EventAudio_Amplitude, B=$NPC.FortniteNPC.EventAudio_Amplitude)
      B=1
      Alpha=$System.MusicReactivePreview
     )
    )
    Index 0 Param 2=5
    Index 0 Param 3=&RandomRangeFloat(Minimum=0, Maximum=0.2)
    Index 0 Param 4=&RandomRangeFloat(Minimum=0, Maximum=0.2)
    Module.Index0Param0Write=true
    Module.Index0Param1Write=true
    Module.Index0Param2Write=true
    Module.Index0Param3Write=true
   )
   &ScaleColor(Scale Alpha=&RandomRangeFloat(Minimum=0.85, Maximum=1), ScaleA=true, ScaleRGB=true)
   &Drag(Drag=5, Use Linear Drag=true, Use Rotational Drag=true)
   &SolveForcesAndVelocity(Clamp Velocity=false, Limit Acceleration=false)
  ]
  rendererClasses [&NiagaraMeshRendererProperties]
  renderers [{rendererClass=&NiagaraMeshRendererProperties, rendererIndex=0}]
 }
 Ring {
  emitterUpdateScript [
   &EmitterState(Inactive Response=::NewEnumerator1, Life Cycle Mode=::NewEnumerator1, Loop Behavior=::NewEnumerator1, UseLoopDelay=false)
   &SpawnBurst_Instantaneous(Use Spawn Probability=true, UseLoopCountLimit=false)
   &SpawnBurst_Instantaneous(Spawn Time=1.1, Use Spawn Probability=true, UseLoopCountLimit=false) {enabled=false, module=SpawnBurst_Instantaneous001}
   &SpawnBurst_Instantaneous(Spawn Time=2.4, Use Spawn Probability=true, UseLoopCountLimit=false) {enabled=false, module=SpawnBurst_Instantaneous002}
  ]
  particleSpawnScript [
   &InitializeParticle(
    Color=<rgba 0 0.8854 0.6675 20>
    Lifetime=0.5
    Mesh Scale Mode=::NewEnumerator3
    Mesh Uniform Scale=0.5
    UsePositionOffset=false
    Write Material Random=true
   )
   &InitialMeshOrientation(Mesh Orientation Mode=::NewEnumerator2, Rotation=<0 0.25 0>, Use Rotation Vector=true)
   &InitialMeshOrientation(Mesh Orientation Mode=::NewEnumerator2, Rotation=<0.05 0 0>, Use Rotation Vector=true) {module=InitialMeshOrientation001}
  ]
  particleUpdateScript [
   &ParticleState()
   &ScaleMeshSize(Scale Factor=&VectorFromFloat(Value=&FloatFromCurve(FloatCurve=DataInterface() {propertyValues {Curve {keys [
          {arriveTangent=2.0253, interpMode=RCIM_Cubic, leaveTangent=2.0253, tangentMode=RCTM_User, time=0, value=0}
          {arriveTangent=1, interpMode=RCIM_Cubic, leaveTangent=1, time=0.2812, value=0.5867}
          {arriveTangent=0.0523, interpMode=RCIM_Cubic, leaveTangent=0.0523, tangentMode=RCTM_User, time=1, value=1}
         ]}}})))
   &ScaleColor(
    Scale Alpha=&FloatFromCurve(FloatCurve=DataInterface() {propertyValues {Curve {keys [
         {arriveTangent=-0.8398, interpMode=RCIM_Cubic, leaveTangent=-0.8398, tangentMode=RCTM_User, time=0, value=1}
         {arriveTangent=-1.084, interpMode=RCIM_Cubic, leaveTangent=-1.084, tangentMode=RCTM_User, time=0.4, value=0.4}
         <at 1 0>
        ]}}})
    Scale RGB=&MakeVectorFromLinearColorRGB(Color=<rgba 1 1 1 1>)
    ScaleA=true
    ScaleRGB=true
   )
   /SetVariables(Particles.GlowColor=&MakeLinearColorFromVectorAndFloat(
     "Vector (RGB)"=&MakeVectorFromLinearColorRGB(Color=&MultiplyLinearColor(Linear Color=<rgba 0.22 1.3902 2 1>, Scale Factor=&MakeVector4(X=2, Y=1, Z=1, W=1)))
     "Float (Alpha)"=&FloatFromCurve(
      FloatCurve=DataInterface() {propertyValues {Curve {keys [
          {arriveTangent=-5.4476, interpMode=RCIM_Cubic, leaveTangent=-5.4476, tangentMode=RCTM_User, time=0, value=1.2}
          {arriveTangent=-1.4541, interpMode=RCIM_Cubic, leaveTangent=-1.4541, tangentMode=RCTM_User, time=0.2913, value=0.3165}
          {arriveTangent=-0.4466, interpMode=RCIM_Cubic, leaveTangent=-0.4466, time=0.6, value=0.05}
          <at 1 0 RCIM_Cubic>
         ]}}}
      Scale Curve=&FN_FrontEndFloatModifier(Gameplay=2, Front End=0.5)
     )
    )) {isSetParameters=true}
   &DynamicMaterialParameters(
    Index 0 Param 1=1.2
    Index 0 Param 2=1
    Index 0 Param 3=1
    Module.Index0Param0Write=true
    Module.Index0Param1Write=true
    Module.Index0Param2Write=true
    Module.Index0Param3Write=true
   )
  ]
  rendererClasses [&NiagaraMeshRendererProperties]
  renderers [
   {rendererClass=&NiagaraMeshRendererProperties, rendererIndex=0}
   {rendererClass=&NiagaraMeshRendererProperties, rendererIndex=1}
  ]
 }
 number {
  emitterUpdateScript [
   &EmitterState(Inactive Response=::NewEnumerator1, Life Cycle Mode=::NewEnumerator1, Loop Behavior=::NewEnumerator1, UseLoopDelay=false)
   &SpawnBurst_Instantaneous(Use Spawn Probability=false, UseLoopCountLimit=false)
   &SpawnBurst_Instantaneous(Spawn Group=1, Spawn Time=1.1, Use Spawn Probability=false, UseLoopCountLimit=false) {enabled=false, module=SpawnBurst_Instantaneous001}
   &SpawnBurst_Instantaneous(Spawn Group=2, Spawn Time=2.4, Use Spawn Probability=false, UseLoopCountLimit=false) {enabled=false, module=SpawnBurst_Instantaneous002}
  ]
  particleSpawnScript [
   &FN_SetSpawnGroup(Number of Spawn Groups=3)
   &InitializeParticle(
    Color=<rgba 0.1771 0.1771 0.1771 1>
    Mesh Renderer Array Visibility Mode=::NewEnumerator1
    Mesh Renderer Info=DataInterface() {propertyValues {MeshRenderer=&self:number_1.NiagaraMeshRendererProperties_0}}
    Mesh Renderer Mesh Index:NiagaraInt32=$Particles.SpawnGroup
    Mesh Scale Mode=::NewEnumerator3
    Mesh Uniform Scale=0.6
    Sprite Size Mode=::NewEnumerator3
    Uniform Sprite Size=150
    UsePositionOffset=false
    Write Material Random=true
   )
   &InitialMeshOrientation(Mesh Orientation Mode=::NewEnumerator2, Rotation=<0 0 -0.25>, Use Rotation Vector=true)
   /SetVariables(Particles.MeshIndex=$User.MeshInt) {isSetParameters=true}
  ]
  particleUpdateScript [
   &ParticleState()
   &ScaleMeshSize(Scale Factor=&VectorFromFloat(Value=&FloatFromCurve(FloatCurve=DataInterface() {propertyValues {Curve {keys [<at 0 0.55>,<at 0.1235 1.1837>,<at 0.2052 0.8>,<at 0.8052 0.8>,<at 0.8653 1.1013>,<at 1 0>]}}})))
   /SetVariables(Particles.Color_glow:LinearColor=&MakeLinearColorFromVectorAndFloat(
     "Vector (RGB)"=&MakeVectorFromLinearColorRGB(Color=&ColorFromCurve(ColorCurve=DataInterface() {propertyValues {
         AlphaCurve {keys []}
         BlueCurve {keys [<at 0.7382 0.785 RCIM_Cubic>,<at 0.9637 1 RCIM_Cubic>]}
         GreenCurve {keys [<at 0.7382 1 RCIM_Cubic>,<at 0.9637 0.7479 RCIM_Cubic>]}
         RedCurve {keys [<at 0.7382 0 RCIM_Cubic>,<at 0.9637 0 RCIM_Cubic>]}
        }}))
     "Float (Alpha)"=&FloatFromCurve(
      FloatCurve=DataInterface() {propertyValues {Curve {keys [<at 0 0>,<at 0.1 1>,<at 0.8 1>,<at 1 0>]}}}
      Scale Curve=&Multiply_Float(
       A=&Multiply_Float(A=0.3, B=&FN_PickFloatDependOnPlatform(TargetPlatformActive=1, TargetPlatformNoActive=0.5))
       B=&FN_FrontEndFloatModifier(Gameplay=1, Front End=0.6)
      )
     )
    )) {isSetParameters=true}
   /SetVariables(Particles.Size_glow:Vector2f=&Vector2DFromFloat(Value=&FloatFromCurve(
      FloatCurve=DataInterface() {propertyValues {Curve {keys [
          {arriveTangent=5.6874, interpMode=RCIM_Cubic, leaveTangent=5.6874, tangentMode=RCTM_User, time=0, value=0.6}
          {arriveTangent=0.0188, interpMode=RCIM_Cubic, leaveTangent=0.0188, tangentMode=RCTM_User, time=0.1, value=1.4}
          {arriveTangent=0.2839, interpMode=RCIM_Cubic, leaveTangent=0.2839, tangentMode=RCTM_User, time=0.2, value=1.2}
         ]}}}
      Scale Curve=&Multiply_Float(
       A=70
       B=&Multiply_Float(A=&FN_FrontEndFloatModifier(Gameplay=1, Front End=0.8), B=1)
      )
     ))) {isSetParameters=true}
   /SetVariables(Particles.Color_dark:LinearColor=<rgba 0 0.0616 0.15 0.6>) {isSetParameters=true}
   &SolveForcesAndVelocity(Clamp Velocity=false, Limit Acceleration=false)
  ]
  rendererClasses [&NiagaraMeshRendererProperties, &NiagaraSpriteRendererProperties]
  renderers [
   {rendererClass=&NiagaraMeshRendererProperties, rendererIndex=0}
   {rendererClass=&NiagaraSpriteRendererProperties, rendererIndex=1}
   {rendererClass=&NiagaraSpriteRendererProperties, rendererIndex=2}
  ]
 }
 particle001 {
  emitterUpdateScript [
   &EmitterState(Inactive Response=::NewEnumerator1, Life Cycle Mode=::NewEnumerator1, Loop Behavior=::NewEnumerator1, UseLoopDelay=false)
   &SpawnBurst_Instantaneous(Spawn Count=15, Use Spawn Probability=false, UseLoopCountLimit=false)
   &SpawnBurst_Instantaneous(Spawn Count=15, Spawn Time=1.1, Use Spawn Probability=false, UseLoopCountLimit=false) {enabled=false, module=SpawnBurst_Instantaneous001}
   &SpawnBurst_Instantaneous(Spawn Count=15, Spawn Time=2.4, Use Spawn Probability=false, UseLoopCountLimit=false) {enabled=false, module=SpawnBurst_Instantaneous002}
  ]
  particleSpawnScript [
   &InitializeParticle(
    Color Maximum=&MakeCustomLinearColorFromBool(Bool=&RandomBool(), True Color=<rgba 0 0.6562 0.3182 1>, False Color=<rgba 0 0.232 1 1>)
    Color Minimum=&MakeCustomLinearColorFromBool(Bool=&RandomBool(), True Color=<rgba 0 1 0.8473 1>, False Color=<rgba 0.3102 0 1 1>)
    Color Mode=::NewEnumerator2
    Lifetime Max=0.4
    Lifetime Min=0.15
    Lifetime Mode=::NewEnumerator1
    Sprite Rotation Angle=&RandomRangeFloat(Minimum=-0.15, Maximum=0.15)
    Sprite Rotation Mode=::NewEnumerator1
    Sprite Size Mode=::NewEnumerator1
    Uniform Sprite Size Max=6
    Uniform Sprite Size Min=4
    UsePositionOffset=false
    Write Material Random=true
   )
   &ShapeLocation(MaskSpawnGroup=false, Override Seed=false, Sphere Radius=20, UseOwnerScale=true)
   &AddVelocity(Module.Use Falloff Distance=false, Override Seed=false, Velocity Mode=::NewEnumerator1, Velocity Speed=100)
   /SetVariables(Particles.VisibilityTag=&RandomRangeInt(Minimum=0, Maximum=2)) {isSetParameters=true}
  ]
  particleUpdateScript [
   &ParticleState()
   &CurlNoiseForce(Bias Noise=false, Noise Strength=100, Pan Noise=false)
   &Drag(Drag=2, Use Linear Drag=true, Use Rotational Drag=true)
   &SpriteRotationRate(Rotation Rate=&Multiply_Float(
     A=&RandomRangeFloat(Minimum=-10, Maximum=10)
     B=&FloatFromCurve(FloatCurve=DataInterface() {propertyValues {Curve {keys [<at 0 1>,<at 1 0>]}}}, Scale Curve=10)
    ))
   &ScaleSpriteSize(
    Uniform Curve Scale=0.6
    Uniform Curve Sprite Scale=DataInterface() {propertyValues {Curve {keys [{tangentMode=RCTM_User, time=0, value=0}, <at 0.2 0.6>, {tangentMode=RCTM_User, time=1, value=1}]}}}
   )
   /SetVariables(Particles.Size002:Vector2f=&Multiply_Vector2D(
     A=$Particles.SpriteSize
     B=&Vector2DFromFloat(Value=&RandomRangeFloat(Minimum=1, Maximum=1))
    )) {isSetParameters=true}
   &ScaleColor(
    Scale Alpha=&FloatFromCurve(FloatCurve=DataInterface() {propertyValues {Curve {keys [<at 0.5 1>,<at 1 0>]}}}, Scale Curve=2)
    ScaleA=true
    ScaleRGB=true
   )
   /SetVariables(Particles.Color_glow:LinearColor=&MakeLinearColorFromVectorAndFloat(
     "Vector (RGB)"=&MakeVectorFromLinearColorRGB(Color=<rgba 0.9598 0 1 1>)
     "Float (Alpha)"=&FloatFromCurve(FloatCurve=DataInterface() {propertyValues {Curve {keys [<at 0.5 1>,<at 1 0>]}}}, Scale Curve=2)
    )) {isSetParameters=true}
   &SolveForcesAndVelocity(Clamp Velocity=false, Limit Acceleration=false)
   &DynamicMaterialParameters(
    Index 0 Param 1=0.2
    Index 0 Param 2=2
    Module.Index0Param0Write=true
    Module.Index0Param1Write=true
    Module.Index0Param2Write=true
    Module.Index0Param3Write=true
   )
  ]
  rendererClasses [&NiagaraSpriteRendererProperties]
  renderers [
   {rendererClass=&NiagaraSpriteRendererProperties}
   {rendererClass=&NiagaraSpriteRendererProperties, rendererIndex=1}
  ]
 }
}
summary {
 systemName=NS_Emote_JadeTowelGloss_Number
 emitters [
  {emitterName=number, rendererClasses [&NiagaraMeshRendererProperties, &NiagaraSpriteRendererProperties]}
  {emitterName=Flare, rendererClasses [&NiagaraMeshRendererProperties]}
  {emitterName=Ring, rendererClasses [&NiagaraMeshRendererProperties]}
  {emitterName=particle001, rendererClasses [&NiagaraSpriteRendererProperties]}
  {emitterName=Minimal004, rendererClasses [&NiagaraMeshRendererProperties]}
  {emitterName=Glitch_sprite002, rendererClasses [&NiagaraMeshRendererProperties]}
  {emitterName=Glitch_sprite, rendererClasses [&NiagaraSpriteRendererProperties]}
 ]
 userVariables [
  User.bIsFrontEnd {type=NiagaraBool}
  User.bIsFrontEndPreview {type=NiagaraBool}
  User.FrontEndSize {type=NiagaraFloat, defaultValue=0.65, description=""}
  User.MeshInt {type=NiagaraInt32}
 ]
}
