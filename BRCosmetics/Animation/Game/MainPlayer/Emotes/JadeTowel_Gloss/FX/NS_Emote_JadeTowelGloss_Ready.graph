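# Niagara system for the "Ready" effect of the JadeTowel Gloss emote, drawing a short flourish in cyan and jade green.
# Box, Triangle and glow emitters spawn on tori for about half their two-second loop, sweeping position by emitter age.
# Box and glow meshes stretch vertically and fade, while Triangle meshes and sprites spin and drift upward.
# Triangle feeds audio amplitude into materials and scales alpha by time-of-day and platform; glow adjusts for front end.
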NiagaraSystem /BRCosmetics/Animation/Game/MainPlayer/Emotes/JadeTowel_Gloss/FX/NS_Emote_JadeTowelGloss_Ready
dependencies {
 usedDataInterfaces [
  &NiagaraDataInterfaceCurve
  &NiagaraDataInterfaceArrayFloat3
  &NiagaraDataInterfaceColorCurve
  &NiagaraDataInterfaceArrayColor
  &NiagaraDataInterfacePlatformSet
 ]
 usedDynamicInputs []
 usedModules [
  &self:SystemSpawnScript.SystemScriptSource.SystemScriptGraph.NiagaraNodeAssignment_0.SetVariables_#1
  &SystemState
 ]
}
emitters {
 Box {
  emitterUpdateScript [
   &EmitterState(Inactive Response=::NewEnumerator1, Life Cycle Mode=::NewEnumerator1, Loop Behavior=::NewEnumerator1, Loop Duration=2, UseLoopDelay=true)
   &SpawnBurst_Instantaneous(Spawn Count=50, Use Spawn Probability=false, UseLoopCountLimit=false) {enabled=false}
   &SpawnRate(
    SpawnRate=&FloatFromCurve(FloatCurve=DataInterface() {propertyValues {Curve {keys [<at 0.5 1>,<at 0.51 0>]}}}, CurveIndex=$Emitter.NormalizedLoopAge, Scale Curve=50)
    Use Spawn Probability=false
   )
  ]
  particleSpawnScript [
   &PartitionParticles(Number Of Partitions=40)
   &InitializeParticle(
    Color=<rgba 0 1 0.8348 1>
    Lifetime=0.8
    Mesh Scale Max=<0.8 0.8 1.5>
    Mesh Scale Min=<0.8 0.8 0.5>
    Mesh Scale Mode=::NewEnumerator2
    UsePositionOffset=false
    Write Material Random=true
   )
   &InitialMeshOrientation(
    Mesh Orientation Mode=::NewEnumerator2
    Rotation=&Multiply_Vector(
     A=&SelectVectorFromArray(Direct Array Index=$Particles.PartitionParticles.Partition, Vector Selection Array=DataInterface() {propertyValues {FloatData [<0 0 0>]}})
     B=&MakeVector(X=0, Y=0, Z=0)
    )
    Use Rotation Vector=true
   )
   /SetVariables(Particles.Randomfloat:NiagaraFloat=&RandomRangeFloat(Minimum=0, Maximum=1)) {isSetParameters=true}
   &ShapeLocation(
    Handle Radius=5
    Large Radius=70
    MaskSpawnGroup=false
    Shape Primitive=::NewEnumerator3
    Torus Distribution Mode:ENiagaraTorusDistributionMode=::NewEnumerator1
    U Position=&FloatFromCurve(FloatCurve=DataInterface() {propertyValues {Curve {keys [<at 0 1.2>,<at 1 0.3>]}}}, CurveIndex=$Emitter.Age)
    UseOwnerScale=true
    V Position=1
   )
  ]
  particleUpdateScript [
   &ParticleState()
   &ScaleMeshSize(Scale Factor=&Multiply_Vector(
     A=&MakeVector(X=1, Y=1, Z=&RandomRangeFloat(Minimum=5, Maximum=15))
     B=&MakeVector(
      X=1
      Y=1
      Z=&Multiply_Float(
       A=&FloatFromCurve(
        FloatCurve=DataInterface() {propertyValues {Curve {keys [
            {arriveTangent=-1.6534, interpMode=RCIM_Cubic, leaveTangent=-1.6534, tangentMode=RCTM_User, time=0, value=1}
            {tangentMode=RCTM_User, time=0.6, value=0}
           ]}}}
        Scale Curve=&RandomRangeFloat(Minimum=0.7, Maximum=1)
       )
       B=1
      )
     )
    ))
   &Color(Color=&LerpLinearColors(StartColor=<rgba 0 1.6021 2.5 1>, EndColor=<rgba 0 1 0.4968 1>, LerpFactor=$Particles.Randomfloat))
   &ScaleColor(
    Color Value To Scale=$Particles.Color
    Scale Alpha=&Multiply_Float(
     A=1
     B=&FloatFromCurve(FloatCurve=DataInterface() {propertyValues {Curve {keys [
          {interpMode=RCIM_Cubic, tangentMode=RCTM_User, time=0, value=0}
          {arriveTangent=-0.029, interpMode=RCIM_Cubic, leaveTangent=-0.029, tangentMode=RCTM_User, time=0.2, value=1}
          {tangentMode=RCTM_User, time=0.5, value=0}
         ]}}})
    )
    Scale RGB=&Multiply_VectorByFloat(
     Vector=&MakeVectorFromLinearColorRGB(Color=&ColorFromCurve(ColorCurve=DataInterface() {propertyValues {
         AlphaCurve {keys [<at 0 1 RCIM_Cubic>,<at 1 0 RCIM_Cubic>]}
         BlueCurve {keys [<at 0 1 RCIM_Cubic>,<at 1 1 RCIM_Cubic>]}
         GreenCurve {keys [<at 0 1 RCIM_Cubic>,<at 1 1 RCIM_Cubic>]}
         RedCurve {keys [<at 0 1 RCIM_Cubic>,<at 1 1 RCIM_Cubic>]}
        }}))
     Float=&RandomRangeFloat(Minimum=0.5, Maximum=1.2)
    )
    ScaleA=true
    ScaleRGB=true
   )
   /SetVariables(Particles.Color002=&MakeLinearColorFromVectorAndFloat(
     "Vector (RGB)"=&MakeVectorFromLinearColorRGB(Color=<rgba 0.0315 1 0.6961 1>)
     "Float (Alpha)"=&Multiply_Float(
      A=0.3
      B=&FloatFromCurve(FloatCurve=DataInterface() {propertyValues {Curve {keys [
           {interpMode=RCIM_Cubic, tangentMode=RCTM_User, time=0, value=0}
           {arriveTangent=-0.029, interpMode=RCIM_Cubic, leaveTangent=-0.029, tangentMode=RCTM_User, time=0.2, value=1}
           {tangentMode=RCTM_User, time=0.5, value=0}
          ]}}})
     )
    )) {isSetParameters=true}
   /SetVariables(Particles.Size002:Vector3f=&Multiply_Vector(
     A=&MakeVector(X=0.8, Y=0.8, Z=&RandomRangeFloat(Minimum=10, Maximum=15))
     B=&MakeVector(
      X=1
      Y=1
      Z=&Multiply_Float(
       A=&FloatFromCurve(
        FloatCurve=DataInterface() {propertyValues {Curve {keys [
            {arriveTangent=-2.1666, interpMode=RCIM_Cubic, leaveTangent=-2.1666, tangentMode=RCTM_User, time=0, value=1}
            {tangentMode=RCTM_User, time=0.6, value=0}
           ]}}}
        Scale Curve=&RandomRangeFloat(Minimum=0.7, Maximum=1)
       )
       B=1
      )
     )
    )) {isSetParameters=true}
   &DynamicMaterialParameters(
    Index 0 Param 1=&MakeFloatFromVector(Vector=&Multiply_Vector(A=<1 1 1>, B=$Particles.Scale), Channel=::NewEnumerator0)
    Index 0 Param 2=&MakeFloatFromVector(Vector=&Multiply_Vector(A=<1 1 1>, B=$Particles.Scale), Channel=::NewEnumerator1)
    Index 0 Param 3=&MakeFloatFromVector(Vector=&Multiply_Vector(A=<1 1 1>, B=$Particles.Scale), Channel=::NewEnumerator2)
    Index 1 Param 1=&Lerp_Float(A=-0.5, B=1, Alpha=$Particles.Randomfloat)
    Module.Index0Param0Write=true
    Module.Index0Param1Write=true
    Module.Index0Param2Write=true
    Module.Index0Param3Write=true
    Module.Index1Param0Write=true
    Module.Index1Param1Write=true
    Module.Index1Param2Write=true
    Module.Index1Param3Write=true
    Write Parameter Index 1=true
   )
  ]
  rendererClasses [&NiagaraMeshRendererProperties]
  renderers [
   {rendererClass=&NiagaraMeshRendererProperties, rendererIndex=0}
   {rendererClass=&NiagaraMeshRendererProperties, rendererIndex=1}
  ]
 }
 Triangle {
  emitterUpdateScript [
   &EmitterState(Inactive Response=::NewEnumerator1, Life Cycle Mode=::NewEnumerator1, Loop Behavior=::NewEnumerator1, Loop Duration=2, UseLoopDelay=false)
   &SpawnRate(
    SpawnRate=&FloatFromCurve(FloatCurve=DataInterface() {propertyValues {Curve {keys [<at 0.45 1>,<at 0.5 0>]}}}, CurveIndex=$Emitter.NormalizedLoopAge, Scale Curve=40)
    Use Spawn Probability=false
   )
  ]
  particleSpawnScript [
   &InitializeParticle(
    Color=&SelectLinearColorFromArray(
     Array Sampling Mode=::NewEnumerator1
     Direct Array Index=&RandomRangeInt(Minimum=0, Maximum=2)
     Color Selection Array=DataInterface() {propertyValues {ColorData [<rgba 0 1.1972 2 1>, <rgba 0 2 1.938 1>, <rgba 0 2 1.0475 1>]}}
     Random Seed=?
    )
    Lifetime Max=0.8
    Lifetime Min=0.3
    Lifetime Mode=::NewEnumerator1
    Mass Max=1.5
    Mass Mode=::NewEnumerator3
    Mesh Scale Mode=::NewEnumerator1
    Mesh Uniform Scale Max=0.6
    Mesh Uniform Scale Min=0.3
    Position Offset=<100 0 20>
    Sprite Rotation Angle=&RandomRangeFloat(Minimum=-0.15, Maximum=0.15)
    Sprite Rotation Mode=::NewEnumerator1
    Sprite Size Mode=::NewEnumerator1
    Uniform Sprite Size Max=4
    Uniform Sprite Size Min=3
    UsePositionOffset=true
    Write Material Random=true
   )
   &InitialMeshOrientation(
    Mesh Orientation Mode=::NewEnumerator2
    Rotation=&RandomRangeVector(Minimum=<0 0 0>, Maximum=<1 1 1>)
    Use Rotation Vector=true
   )
   &ShapeLocation(
    Handle Radius=20
    Large Radius=50
    MaskSpawnGroup=false
    Shape Primitive=::NewEnumerator3
    Torus Distribution Mode:ENiagaraTorusDistributionMode=::NewEnumerator1
    U Position=&FloatFromCurve(FloatCurve=DataInterface() {propertyValues {Curve {keys [<at 0 1.2>,<at 1 0.3>]}}}, CurveIndex=$Emitter.Age)
    UseOwnerScale=true
    V Position=1
   )
   &AddVelocity(Velocity=&RandomRangeVector(Minimum=<0 0 30>, Maximum=<0 0 300>))
   /SetVariables(Particles.VisibilityTag=&RandomRangeInt(Minimum=0, Maximum=1)) {isSetParameters=true}
  ]
  particleUpdateScript [
   &ParticleState()
   &SpriteRotationRate(Rotation Rate=&Multiply_Float(
     A=&RandomRangeFloat(Minimum=-20, Maximum=20)
     B=&FloatFromCurve(FloatCurve=DataInterface() {propertyValues {Curve {keys [<at 0 1>,<at 1 0.45>]}}}, Scale Curve=10)
    ))
   &ScaleSpriteSize(
    Uniform Curve Scale=&RandomRangeFloat(Minimum=1, Maximum=1.5)
    Uniform Curve Sprite Scale=DataInterface() {propertyValues {Curve {keys [{tangentMode=RCTM_User, time=0, value=0.4}, <at 0.1 1.2>, <at 0.2 1>, {tangentMode=RCTM_User, time=1, value=1}]}}}
   )
   /SetVariables(Particles.Size003:Vector2f=&Multiply_Vector2D(
     A=&Vector2DFromFloat(Value=20)
     B=&Vector2DFromFloat(Value=&FloatFromCurve(FloatCurve=DataInterface() {propertyValues {Curve {keys [<at 0 0.6>,<at 0.2 1>,<at 0.8 1>,<at 1 0.6>]}}}))
    )) {isSetParameters=true}
   /SetVariables(Particles.Color003:LinearColor=&MakeLinearColorFromVectorAndFloat(
     "Vector (RGB)"=&MakeVectorFromLinearColorRGB(Color=$Particles.Color)
     "Float (Alpha)"=&FloatFromCurve(
      FloatCurve=DataInterface() {propertyValues {Curve {keys [<at 0 0 RCIM_Cubic>,<at 0.2 1 RCIM_Cubic>,<at 1 0 RCIM_Cubic>]}}}
      Scale Curve=&Multiply_Float(
       A=0.12
       B=&Multiply_Float(
        A=&FNTimeOfDayScalar(Night Scale=0.6, Morning Scale=1, Evening Scale=1)
        B=&FN_PickFloatDependOnPlatform(TargetPlatformActive=1, TargetPlatformNoActive=0.5)
       )
      )
     )
    )) {isSetParameters=true}
   /SetVariables(Particles.Dyn002:Vector4f=&MakeVector4(X=2, Y=1, Z=0, W=0)) {isSetParameters=true}
   &UpdateMeshOrientation(
    Rotation Rate=&RandomRangeFloat(Minimum=0.5, Maximum=1.2)
    Rotation Vector=&RandomRangeVector(Minimum=<0.2 0.2 0.2>, Maximum=<0.6 0.6 0.6>)
   )
   &ScaleMeshSize(Scale Factor=&VectorFromFloat(Value=&FloatFromCurve(FloatCurve=DataInterface() {propertyValues {Curve {keys [
          {interpMode=RCIM_Cubic, tangentMode=RCTM_User, time=0, value=0}
          {interpMode=RCIM_Cubic, tangentMode=RCTM_User, time=0.2, value=1}
          {arriveTangent=-0.0093, interpMode=RCIM_Cubic, leaveTangent=-0.0093, tangentMode=RCTM_User, time=0.5, value=1}
          {tangentMode=RCTM_User, time=1, value=0}
         ]}}})))
   &DynamicMaterialParameters(
    Index 0 Param 1=&Add_Float(
     A=20
     B=&Lerp_Float(
      A=&Multiply_Float(A=$NPC.FortniteNPC.EventAudio_Amplitude, B=$NPC.FortniteNPC.EventAudio_Amplitude)
      B=1
      Alpha=$System.MusicReactivePreview
     )
    )
    Index 0 Param 2=5
    Index 0 Param 3=&RandomRangeFloat(Minimum=0, Maximum=0.2)
    Index 0 Param 4=&RandomRangeFloat(Minimum=0, Maximum=0.2)
    Module.Index0Param0Write=true
    Module.Index0Param1Write=true
    Module.Index0Param2Write=true
    Module.Index0Param3Write=true
   )
   &ScaleColor(Scale Alpha=&RandomRangeFloat(Minimum=0.85, Maximum=1), ScaleA=true, ScaleRGB=true)
   &Drag(Use Linear Drag=true, Use Rotational Drag=true)
   &SolveForcesAndVelocity(Clamp Velocity=false, Limit Acceleration=false)
  ]
  rendererClasses [&NiagaraMeshRendererProperties, &NiagaraSpriteRendererProperties]
  renderers [
   {rendererClass=&NiagaraMeshRendererProperties, rendererIndex=0}
   {rendererClass=&NiagaraSpriteRendererProperties, rendererIndex=1}
  ]
 }
 glow {
  emitterUpdateScript [
   &EmitterState(Inactive Response=::NewEnumerator1, Life Cycle Mode=::NewEnumerator1, Loop Behavior=::NewEnumerator1, Loop Duration=2, UseLoopDelay=false)
   &SpawnRate(
    SpawnRate=&FloatFromCurve(FloatCurve=DataInterface() {propertyValues {Curve {keys [<at 0.5 1>,<at 0.51 0>]}}}, CurveIndex=$Emitter.NormalizedLoopAge, Scale Curve=20)
    Use Spawn Probability=false
   )
  ]
  particleSpawnScript [
   &InitializeParticle(
    Color Maximum=<rgba 2 2 2 1>
    Color Minimum=<rgba 1 1 1 1>
    Color Mode=::NewEnumerator2
    Lifetime Max=0.6
    Lifetime Min=0.5
    Lifetime Mode=::NewEnumerator1
    Mesh Scale Max=<3 3 10>
    Mesh Scale Min=<2 2 5>
    Mesh Scale Mode=::NewEnumerator2
    Position Offset=<100 0 0>
    UsePositionOffset=true
    Write Material Random=true
   )
   &InitialMeshOrientation(
    Mesh Orientation Mode=::NewEnumerator2
    Rotation=&RandomRangeVector(Minimum=<0 0 0>, Maximum=<0 0 1>)
    Use Rotation Vector=true
   )
   &ShapeLocation(
    Handle Radius=20
    Large Radius=50
    MaskSpawnGroup=false
    Shape Primitive=::NewEnumerator3
    Torus Distribution Mode:ENiagaraTorusDistributionMode=::NewEnumerator1
    U Position=&FloatFromCurve(FloatCurve=DataInterface() {propertyValues {Curve {keys [<at 0 1.2>,<at 1 0.3>]}}}, CurveIndex=$Emitter.Age)
    UseOwnerScale=true
    V Position=1
   )
  ]
  particleUpdateScript [
   &ParticleState()
   &ScaleColor(
    Scale Alpha=&FloatFromCurve(
     FloatCurve=DataInterface() {propertyValues {Curve {keys [
         <at 0 0 RCIM_Cubic>
         {arriveTangent=0.037, interpMode=RCIM_Cubic, leaveTangent=0.037, tangentMode=RCTM_User, time=0.4, value=1}
         <at 1 0 RCIM_Cubic>
        ]}}}
     Scale Curve=&FN_FrontEndFloatModifier(Gameplay=5, Front End=3)
    )
    Scale RGB=&MakeVectorFromLinearColorRGB(Color=&ColorFromCurve(ColorCurve=DataInterface() {propertyValues {
        AlphaCurve {keys []}
        BlueCurve {keys [<at 0.1903 0.624>,<at 0.4323 1>,<at 0.6335 1>,<at 0.8508 0.5515 RCIM_Cubic>]}
        GreenCurve {keys [<at 0.1903 1>,<at 0.4323 0.1914>,<at 0.6335 0.7362>,<at 0.8508 1 RCIM_Cubic>]}
        RedCurve {keys [<at 0.1903 0>,<at 0.4323 0>,<at 0.6335 0>,<at 0.8508 0 RCIM_Cubic>]}
       }}))
    ScaleA=true
    ScaleRGB=true
   )
   &ScaleMeshSize(Scale Factor=&MakeVector(
     X=1
     Y=1
     Z=&FloatFromCurve(FloatCurve=DataInterface() {propertyValues {Curve {keys [<at 0 0>,<at 0.5 1>,<at 1 0>]}}})
    ))
   &SolveForcesAndVelocity(Clamp Velocity=false, Limit Acceleration=false)
  ]
  rendererClasses [&NiagaraMeshRendererProperties]
  renderers [{rendererClass=&NiagaraMeshRendererProperties, rendererIndex=0}]
 }
}
summary {
 systemName=NS_Emote_JadeTowelGloss_Ready
 emitters [
  {emitterName=Box, rendererClasses [&NiagaraMeshRendererProperties]}
  {emitterName=glow, rendererClasses [&NiagaraMeshRendererProperties]}
  {emitterName=Triangle, rendererClasses [&NiagaraMeshRendererProperties, &NiagaraSpriteRendererProperties]}
 ]
 userVariables [
  User.bIsFrontEnd {type=NiagaraBool}
  User.bIsFrontEndPreview {type=NiagaraBool}
 ]
}
