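# Niagara system for the JadeTowel Gloss emote's searchlight effect, a looping, music-reactive stage show around the player.
# Emitters layer rotating stage lights, a backboard with a dynamic light, stars, box bars with glows, rings and drifting triangles, often placed through partition-indexed arrays.
# Event audio amplitude drives star, box and triangle scaling, with a preview override for some.
# Brightness and size adapt to front-end preview, time of day, distance and platform.
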NiagaraSystem /BRCosmetics/Animation/Game/MainPlayer/Emotes/JadeTowel_Gloss/FX/NS_Emote_JadeTowelGloss_Searchlight
dependencies {
 usedDataInterfaces [
  &NiagaraDataInterfaceArrayFloat3
  &NiagaraDataInterfaceColorCurve
  &NiagaraDataInterfaceCurve
  &NiagaraDataInterfaceArrayColor
  &NiagaraDataInterfaceArrayFloat
  &NiagaraDataInterfacePlatformSet
 ]
 usedDynamicInputs [&RemapRange]
 usedModules [
  &self:SystemSpawnScript.SystemScriptSource.SystemScriptGraph.NiagaraNodeAssignment_0.SetVariables_#1
  &self:SystemSpawnScript.SystemScriptSource.SystemScriptGraph.NiagaraNodeAssignment_1.SetVariables_#2
  &SystemState
 ]
}
emitters {
 BigStar {
  emitterUpdateScript [
   &EmitterState(Inactive Response=::NewEnumerator1, Life Cycle Mode=::NewEnumerator1, Loop Behavior=::NewEnumerator1, Loop Delay=0.3, UseLoopDelay=true)
   &SpawnBurst_Instantaneous(Spawn Count=6, Use Spawn Probability=false, UseLoopCountLimit=false)
  ]
  particleSpawnScript [
   &PartitionParticles(Number Of Partitions=6)
   &InitializeParticle(
    Color=<rgba 1 1 1 1.5>
    Lifetime=4
    Mesh Scale=<1 8 8>
    Mesh Scale Mode=::NewEnumerator4
    Sprite Size Mode=::NewEnumerator3
    Uniform Sprite Size=&SelectFloatFromArray(
     Direct Array Index=$Particles.PartitionParticles.Partition
     Float Selection Array=DataInterface() {propertyValues {FloatData [30,50,45,40,55,35]}}
    )
    UsePositionOffset=true
    Write Material Random=true
   )
   &ShapeLocation(
    MaskSpawnGroup=false
    Offset=&MakeCustomVectorFromBool(
     Bool=$User.bIsFrontEndPreview
     True Vector=&AddVectorToPosition(
      Position=&SelectVectorFromArray(
       Direct Array Index=$Particles.PartitionParticles.Partition
       Vector Selection Array=DataInterface() {propertyValues {FloatData [<-70 -50 120>, <-110 0 150>, <-60 80 70>, <90 -40 125>, <100 0 110>, <80 50 100>]}}
      )
      Vector=<0 0 0>
     )
     False Vector=&AddVectorToPosition(
      Position=&SelectVectorFromArray(
       Direct Array Index=$Particles.PartitionParticles.Partition
       Vector Selection Array=DataInterface() {propertyValues {FloatData [<-70 -50 120>, <-110 0 150>, <-60 80 100>, <100 -40 110>, <100 0 110>, <80 50 100>]}}
      )
      Vector=<0 0 0>
     )
    )
    Offset Mode=::NewEnumerator0
    Override Seed=false
    Shape Origin=$Particles.Position
    Sphere Radius=0.1
    UseOwnerScale=true
   )
  ]
  particleUpdateScript [
   &ParticleState()
   &ScaleSpriteSize(
    Uniform Curve Scale=&Multiply_Float(
     A=&Multiply_Float(
      A=&RandomRangeFloat(Minimum=0.8, Maximum=1.5)
      B=&ClampFloat(Float=$NPC.FortniteNPC.EventAudio_Amplitude, Min=0.5)
     )
     B=&FN_LOD_Distance_Remap(Start Distance=240, End Distance=600, Minimum Output=0.6, Maximum Output=1.2, Clamp Output=true)
    )
    Uniform Curve Sprite Scale=DataInterface() {propertyValues {Curve {keys [{tangentMode=RCTM_User, time=0, value=0.4}, <at 0.02 1.3>, <at 0.05 1>, {tangentMode=RCTM_User, time=1, value=1}]}}}
   )
   /SetVariables(Particles.Size003:Vector2f=&Multiply_Vector2D(A=$Particles.SpriteSize, B=&Vector2DFromFloat(Value=1.2))) {isSetParameters=true}
   &SpriteRotationRate(Rotation Rate=&Multiply_Float(
     A=&MakeCustomFloatFromBool(Bool=&RandomBool(), True Float=-50, False Float=50)
     B=&RandomRangeFloat(Minimum=0.5, Maximum=1)
    ))
   &ScaleColor(
    Scale Alpha=&FloatFromCurve(
     FloatCurve=DataInterface() {propertyValues {Curve {keys [
         <at 0 0>
         {arriveTangent=-2.0518, leaveTangent=-2.0518, time=0.05, value=1}
         {arriveTangent=-1.1111, leaveTangent=-1.1111, time=0.9, value=1}
         <at 1 0>
        ]}}}
     Scale Curve=&Multiply_Float(
      A=&FNTimeOfDayScalar(Night Scale=0.8, Morning Scale=1, Evening Scale=1)
      B=&Multiply_Float(A=&FN_FrontEndFloatModifier(Gameplay=1, Front End=0.6), B=1)
     )
    )
    Scale RGB=&Multiply_VectorByFloat(
     Vector=&MakeVectorFromLinearColorRGB(Color=&SelectLinearColorFromArray(
       Array Sampling Mode=::NewEnumerator1
       Direct Array Index=$Particles.PartitionParticles.Partition
       Color Selection Array=DataInterface() {propertyValues {ColorData [
          <rgba 0.0496 0.2661 1 1>
          <rgba 0 0.8 0.6337 1>
          <rgba 0 0.5 0.3843 1>
          <rgba 0.0032 0.5781 0.151 1.2>
          <rgba 0.0306 0.7201 0.75 1>
          <rgba 0 0.5369 1 1>
         ]}}
       Random Seed=?
      ))
     Float=1
    )
    ScaleA=true
    ScaleRGB=true
   )
   /SetVariables(Particles.Color002=&MakeLinearColorFromVectorAndFloat(
     "Vector (RGB)"=&MakeVectorFromLinearColorRGB(Color=&ColorFromCurve(ColorCurve=DataInterface() {propertyValues {
         AlphaCurve {keys []}
         BlueCurve {keys [{interpMode=RCIM_Cubic, tangentMode=RCTM_User, time=0.5651, value=0.4819}]}
         GreenCurve {keys [{interpMode=RCIM_Cubic, tangentMode=RCTM_User, time=0.5651, value=1}]}
         RedCurve {keys [{interpMode=RCIM_Cubic, tangentMode=RCTM_User, time=0.5651, value=0.2344}]}
        }}))
     "Float (Alpha)"=&FloatFromCurve(FloatCurve=DataInterface() {propertyValues {Curve {keys [
          <at 0 0>
          {arriveTangent=-2.0518, leaveTangent=-2.0518, time=0.05, value=1}
          {arriveTangent=-1.1111, leaveTangent=-1.1111, time=0.9, value=1}
          <at 1 0>
         ]}}})
    )) {isSetParameters=true}
   /SetVariables(Particles.Color003:LinearColor=&MultiplyLinearColor(
     Linear Color=$Particles.Color
     Scale Factor=&MakeVector4_FromFloat(Float=&FN_FrontEndFloatModifier(Gameplay=1, Front End=0.5))
    )) {isSetParameters=true}
   &SolveForcesAndVelocity(Clamp Velocity=false, Limit Acceleration=false)
   &DynamicMaterialParameters(
    Index 0 Param 1=&SelectFloatFromArray(
     Direct Array Index=$Particles.PartitionParticles.Partition
     Float Selection Array=DataInterface() {propertyValues {FloatData [0.48,1,-2,-2,-2,0.5]}}
    )
    Index 0 Param 2=&SelectFloatFromArray(
     Direct Array Index=$Particles.PartitionParticles.Partition
     Float Selection Array=DataInterface() {propertyValues {FloatData [0,0.3,0,0,0.3,0.3]}}
    )
    Index 0 Param 3=&Multiply_Float(
     A=&FloatFromCurve(FloatCurve=DataInterface() {propertyValues {Curve {keys [<at 0 1>,<at 1 1>]}}}, Scale Curve=50)
     B=&SelectFloatFromArray(
      Direct Array Index=$Particles.PartitionParticles.Partition
      Float Selection Array=DataInterface() {propertyValues {FloatData [0.5,-1,0.65,-0.5,1,-0.65]}}
     )
    )
    Module.Index0Param0Write=true
    Module.Index0Param1Write=true
    Module.Index0Param2Write=true
    Module.Index0Param3Write=true
   )
   /SetVariables(Particles.Dyn002:Vector4f=&MakeVector4(
     X=-2
     Y=0
     Z=&Multiply_Float(
      A=&FloatFromCurve(FloatCurve=DataInterface() {propertyValues {Curve {keys [<at 0 1>,<at 1 1>]}}}, Scale Curve=50)
      B=&SelectFloatFromArray(
       Direct Array Index=$Particles.PartitionParticles.Partition
       Float Selection Array=DataInterface() {propertyValues {FloatData [0.5,-1,0.65,-0.5,1,-0.65]}}
      )
     )
     W=0
    )) {isSetParameters=true}
  ]
  rendererClasses [&NiagaraSpriteRendererProperties]
  renderers [
   {rendererClass=&NiagaraSpriteRendererProperties}
   {rendererClass=&NiagaraSpriteRendererProperties, rendererIndex=1}
   {rendererClass=&NiagaraSpriteRendererProperties, rendererIndex=2}
  ]
 }
 Box {
  emitterUpdateScript [
   &EmitterState(
    Inactive Response=::NewEnumerator1
    Life Cycle Mode=::NewEnumerator1
    Loop Behavior=::NewEnumerator1
    Loop Delay=0.5
    Loop Duration=4
    UseLoopDelay=true
   )
   &SpawnBurst_Instantaneous(Spawn Count=40, Use Spawn Probability=false, UseLoopCountLimit=false)
  ]
  particleSpawnScript [
   &PartitionParticles(Number Of Partitions=40)
   &InitializeParticle(
    Color=<rgba 0 1 0.8348 1>
    Lifetime Max=1.5
    Lifetime Min=0.5
    Lifetime Mode=::NewEnumerator1
    Mesh Scale Max=<0.8 0.8 2>
    Mesh Scale Min=<0.8 0.8 1>
    Mesh Scale Mode=::NewEnumerator2
    UsePositionOffset=false
    Write Material Random=true
   )
   &InitialMeshOrientation(
    Mesh Orientation Mode=::NewEnumerator2
    Rotation=&Multiply_Vector(
     A=&SelectVectorFromArray(
      Direct Array Index=$Particles.PartitionParticles.Partition
      Vector Selection Array=DataInterface() {propertyValues {FloatData [
         <0 0 0>
         <0 0 0.01>
         <0 0 0.02>
         <0 0 0.03>
         <0 0 0.04>
         <0 0 0.05>
         <0 0 0.06>
         <0 0 0.07>
         <0 0 0.08>
         <0 0 0.09>
         <0 0 0>
         <0 0 -0.01>
         <0 0 -0.02>
         <0 0 -0.03>
         <0 0 -0.04>
         <0 0 -0.05>
         <0 0 -0.06>
         <0 0 -0.07>
         <0 0 -0.08>
         <0 0 -0.09>
         <0 0 0.4>
         <0 0 0.41>
         <0 0 0.42>
         <0 0 0.43>
         <0 0 0.44>
         <0 0 0.45>
         <0 0 0.46>
         <0 0 0.47>
         <0 0 0.48>
         <0 0 0.49>
         <0 0 0.5>
         <0 0 0.39>
         <0 0 0.38>
         <0 0 0.37>
         <0 0 0.36>
         <0 0 0.35>
         <0 0 0.34>
         <0 0 0.33>
         <0 0 0.32>
        ]}}
     )
     B=<1 1 1.2>
    )
    Use Rotation Vector=true
   )
   /SetVariables(Particles.Randomfloat:NiagaraFloat=&RandomRangeFloat(Minimum=0, Maximum=1)) {isSetParameters=true}
  ]
  particleUpdateScript [
   &ParticleState(Kill Particles When Lifetime Has Elapsed=false, Loop Particles Lifetime=true)
   &ScaleMeshSize(Scale Factor=&Multiply_Vector(
     A=&MakeVector(X=1, Y=1, Z=&RandomRangeFloat(Minimum=5, Maximum=10))
     B=&MakeVector(
      X=1
      Y=1
      Z=&Multiply_Float(
       A=&FloatFromCurve(
        FloatCurve=DataInterface() {propertyValues {Curve {keys [
            {interpMode=RCIM_Cubic, tangentMode=RCTM_User, time=0, value=0.4}
            {interpMode=RCIM_Cubic, tangentMode=RCTM_User, time=0.5, value=1.2}
            {tangentMode=RCTM_User, time=1, value=0.4}
           ]}}}
        Scale Curve=&RandomRangeFloat(Minimum=0.7, Maximum=1)
       )
       B=&Lerp_Float(
        A=&Multiply_Float(A=$NPC.FortniteNPC.EventAudio_Amplitude, B=$NPC.FortniteNPC.EventAudio_Amplitude)
        B=1
        Alpha=$System.MusicReactivePreview
       )
      )
     )
    ))
   &Color(Color=&LerpLinearColors(StartColor=<rgba 0 1.2817 2 1>, EndColor=<rgba 0 1 0.4968 1>, LerpFactor=$Particles.Randomfloat))
   &ScaleColor(
    Color Value To Scale=$Particles.Color
    Scale Alpha=&Multiply_Float(
     A=&Multiply_Float(
      A=&Multiply_Float(A=1, B=&FN_FrontEndFloatModifier(Gameplay=1, Front End=0.8))
      B=&FNTimeOfDayScalar(Night Scale=0.7, Morning Scale=1, Evening Scale=1)
     )
     B=&FloatFromCurve(
      FloatCurve=DataInterface() {propertyValues {Curve {keys [
          {interpMode=RCIM_Cubic, tangentMode=RCTM_User, time=0, value=0}
          {interpMode=RCIM_Cubic, tangentMode=RCTM_User, time=0.5, value=1}
          {tangentMode=RCTM_User, time=1, value=0}
         ]}}}
      Scale Curve=&MakeCustomFloatFromBool(Bool=&SetBoolByFloatComparison(A=$NPC.FortniteNPC.EventAudio_Amplitude, B=0), True Float=1, False Float=0)
     )
    )
    Scale RGB=&Multiply_VectorByFloat(
     Vector=&MakeVectorFromLinearColorRGB(Color=&ColorFromCurve(ColorCurve=DataInterface() {propertyValues {
         AlphaCurve {keys []}
         BlueCurve {keys [
           <at 0 1 RCIM_Cubic>
           <at 0.1479 3.8798>
           {arriveTangent=-2.4295, interpMode=RCIM_Cubic, leaveTangent=-2.4295, time=0.3483, value=1.3807}
           <at 0.51 3>
          ]}
         GreenCurve {keys [
           <at 0 1 RCIM_Cubic>
           <at 0.1479 5>
           {arriveTangent=-9.785, interpMode=RCIM_Cubic, leaveTangent=-9.785, time=0.3483, value=0.0708}
           <at 0.51 1.4565>
          ]}
         RedCurve {keys [
           <at 0 1 RCIM_Cubic>
           <at 0.1479 1.8663>
           {arriveTangent=-4.7454, interpMode=RCIM_Cubic, leaveTangent=-4.7454, time=0.3483, value=5}
           <at 0.51 0.1478>
          ]}
        }}))
     Float=&RandomRangeFloat(Minimum=0.5, Maximum=&FN_FrontEndFloatModifier(Gameplay=1.2, Front End=0.6))
    )
    ScaleA=true
    ScaleRGB=true
   )
   /SetVariables(Particles.Color002=&Multiply_LinearColorByFloat(
     Linear Color=<rgba 0.0315 1 0.6961 0.2>
     Float=&Multiply_Float(
      A=&FNTimeOfDayScalar(Night Scale=0.8, Morning Scale=1, Evening Scale=1)
      B=&Multiply_Float(A=&FN_FrontEndFloatModifier(Gameplay=1, Front End=0.7), B=1)
     )
    )) {isSetParameters=true}
   /SetVariables(Particles.Size002:Vector3f=&Multiply_Vector(
     A=&MakeVector(X=0.8, Y=0.8, Z=&RandomRangeFloat(Minimum=13, Maximum=16))
     B=&MakeVector(
      X=1
      Y=1
      Z=&Multiply_Float(
       A=&FloatFromCurve(
        FloatCurve=DataInterface() {propertyValues {Curve {keys [
            {interpMode=RCIM_Cubic, tangentMode=RCTM_User, time=0, value=0.4}
            {interpMode=RCIM_Cubic, tangentMode=RCTM_User, time=0.5, value=1.2}
            {tangentMode=RCTM_User, time=1, value=0.4}
           ]}}}
        Scale Curve=&RandomRangeFloat(Minimum=0.7, Maximum=1)
       )
       B=&Lerp_Float(
        A=&Multiply_Float(A=$NPC.FortniteNPC.EventAudio_Amplitude, B=$NPC.FortniteNPC.EventAudio_Amplitude)
        B=1
        Alpha=$System.MusicReactivePreview
       )
      )
     )
    )) {isSetParameters=true}
   &DynamicMaterialParameters(
    Index 0 Param 1=&MakeFloatFromVector(Vector=&Multiply_Vector(A=<1 1 1>, B=$Particles.Scale), Channel=::NewEnumerator0)
    Index 0 Param 2=&MakeFloatFromVector(Vector=&Multiply_Vector(A=<1 1 1>, B=$Particles.Scale), Channel=::NewEnumerator1)
    Index 0 Param 3=&MakeFloatFromVector(Vector=&Multiply_Vector(A=<1 1 1>, B=$Particles.Scale), Channel=::NewEnumerator2)
    Index 1 Param 1=&Lerp_Float(A=-0.5, B=1, Alpha=$Particles.Randomfloat)
    Module.Index0Param0Write=true
    Module.Index0Param1Write=true
    Module.Index0Param2Write=true
    Module.Index0Param3Write=true
    Module.Index1Param0Write=true
    Module.Index1Param1Write=true
    Module.Index1Param2Write=true
    Module.Index1Param3Write=true
    Write Parameter Index 1=true
   )
  ]
  rendererClasses [&NiagaraMeshRendererProperties]
  renderers [
   {rendererClass=&NiagaraMeshRendererProperties, rendererIndex=0}
   {rendererClass=&NiagaraMeshRendererProperties, rendererIndex=1}
  ]
 }
 BoxGlow {
  emitterUpdateScript [
   &EmitterState(Inactive Response=::NewEnumerator1, Life Cycle Mode=::NewEnumerator1, Loop Behavior=::NewEnumerator1, Loop Duration=4, UseLoopDelay=true)
   &SpawnRate(SpawnRate=5, Use Spawn Probability=false)
   &SpawnRate(SpawnGroup=1, SpawnRate=5, Use Spawn Probability=false) {module=SpawnRate001}
  ]
  particleSpawnScript [
   &FN_SetSpawnGroup(Number of Spawn Groups=2)
   &InitializeParticle(
    Color Maximum=<rgba 2 2 2 1>
    Color Minimum=<rgba 1 1 1 1>
    Color Mode=::NewEnumerator2
    Lifetime Max=1.5
    Lifetime Mode=::NewEnumerator1
    Mesh Scale Max=<3 3 8>
    Mesh Scale Min=<2 2 4>
    Mesh Scale Mode=::NewEnumerator2
    Sprite Rotation Mode=::NewEnumerator2
    Sprite Size Mode=::NewEnumerator1
    Uniform Sprite Size Max=15
    Uniform Sprite Size Min=10
    UsePositionOffset=false
    Write Material Random=true
   )
   &ShapeLocation(
    Cylinder Height=50
    Cylinder Radius=30
    MaskSpawnGroup=true
    Non Uniform Scale=<1 1.5 1>
    Offset=<100 0 0>
    Offset Mode=::NewEnumerator0
    Override Seed=false
    Shape Primitive=::NewEnumerator1
    Surface Only=false
    UseOwnerScale=true
   )
   &ShapeLocation(
    Cylinder Height=50
    Cylinder Radius=30
    MaskSpawnGroup=true
    Non Uniform Scale=<1 1.5 1>
    Offset=<-100 0 0>
    Offset Mode=::NewEnumerator0
    Override Seed=false
    Shape Primitive=::NewEnumerator1
    Spawn Group Mask=1
    Surface Only=false
    UseOwnerScale=true
   ) {module=ShapeLocation001}
   &InitialMeshOrientation(
    Mesh Orientation Mode=::NewEnumerator2
    Rotation=&RandomRangeVector(Minimum=<0 0 0>, Maximum=<0 0 1>)
    Use Rotation Vector=true
   )
   &CameraOffset(Camera Offset Amount=20)
  ]
  particleUpdateScript [
   &ParticleState()
   &ScaleColor(
    Scale Alpha=&FloatFromCurve(
     FloatCurve=DataInterface() {propertyValues {Curve {keys [
         <at 0 0 RCIM_Cubic>
         {arriveTangent=0.037, interpMode=RCIM_Cubic, leaveTangent=0.037, tangentMode=RCTM_User, time=0.3, value=1}
         {arriveTangent=-1.7625, interpMode=RCIM_Cubic, leaveTangent=-1.7625, tangentMode=RCTM_User, time=0.6, value=0.4}
         <at 1 0 RCIM_Cubic>
        ]}}}
     Scale Curve=&Multiply_Float(
      A=&Multiply_Float(A=15, B=&FNTimeOfDayScalar(Night Scale=0.8, Morning Scale=1, Evening Scale=1))
      B=&Multiply_Float(A=&FN_FrontEndFloatModifier(Gameplay=1, Front End=0.5), B=1)
     )
    )
    Scale RGB=&MakeVectorFromLinearColorRGB(Color=&ColorFromCurve(ColorCurve=DataInterface() {propertyValues {
        AlphaCurve {keys []}
        BlueCurve {keys [<at 0.1903 0.624>,<at 0.4323 1>,<at 0.5242 1>,<at 0.8508 0.5515 RCIM_Cubic>]}
        GreenCurve {keys [<at 0.1903 1>,<at 0.4323 0.1914>,<at 0.5242 0.0819>,<at 0.8508 1 RCIM_Cubic>]}
        RedCurve {keys [<at 0.1903 0>,<at 0.4323 0>,<at 0.5242 0>,<at 0.8508 0 RCIM_Cubic>]}
       }}))
    ScaleA=true
    ScaleRGB=true
   )
   &SolveForcesAndVelocity(Clamp Velocity=false, Limit Acceleration=false)
  ]
  rendererClasses [&NiagaraMeshRendererProperties]
  renderers [{rendererClass=&NiagaraMeshRendererProperties, rendererIndex=0}]
 }
 Loop_StageLights {
  emitterUpdateScript [
   &EmitterState(
    Inactive Response=::NewEnumerator1
    Life Cycle Mode=::NewEnumerator1
    Loop Behavior=::NewEnumerator1
    Loop Delay=0.6
    Loop Duration=4.5
    UseLoopDelay=true
   )
   &SpawnBurst_Instantaneous(Spawn Count=6, Use Spawn Probability=false, UseLoopCountLimit=false)
  ]
  particleSpawnScript [
   &PartitionParticles(Number Of Partitions=6)
   &InitializeParticle(
    Color Mode=::NewEnumerator0
    Lifetime=4
    Mesh Scale=<0.3 0.3 1>
    Mesh Scale Mode=::NewEnumerator4
    Sprite Size Mode=::NewEnumerator3
    UsePositionOffset=false
    Write Material Random=true
   )
   &ShapeLocation(
    MaskSpawnGroup=false
    Offset=&SelectVectorFromArray(
     Direct Array Index=$Particles.PartitionParticles.Partition
     Vector Selection Array=DataInterface() {propertyValues {FloatData [<90 100 20>, <150 0 20>, <90 -100 20>, <-90 -100 20>, <-150 0 20>, <-90 100 20>]}}
    )
    Offset Mode=::NewEnumerator0
    Override Seed=false
    Sphere Radius=0.1
    UseOwnerScale=true
   )
   &InitialMeshOrientation(
    Mesh Orientation Mode=::NewEnumerator2
    Rotation=&SelectVectorFromArray(
     Direct Array Index=$Particles.PartitionParticles.Partition
     Vector Selection Array=DataInterface() {propertyValues {FloatData [<0.5 0 0.1>, <0.5 0 -0.1>, <0.5 0 0.1>, <0.5 0 -0.1>, <0.5 0 0.1>, <0.5 0 -0.1>]}}
    )
    Use Rotation Vector=true
   )
  ]
  particleUpdateScript [
   &ParticleState()
   &UpdateMeshOrientation(
    Rotation Coordinate Space=::NewEnumerator2
    Rotation Vector=&SelectVectorFromArray(
     Direct Array Index=$Particles.PartitionParticles.Partition
     Vector Selection Array=DataInterface() {propertyValues {FloatData [<0 0 0.5>, <0 0 0.5>, <0 0 0.5>, <0 0 0.5>, <0 0 0.5>, <0 0 0.5>]}}
    )
   )
   &DynamicMaterialParameters(
    Index 0 Param 1=0.15
    Module.Index0Param0Write=true
    Module.Index0Param1Write=true
    Module.Index0Param2Write=true
    Module.Index0Param3Write=true
   )
   /SetVariables(
    Particles.RingPosition:NiagaraPosition=&AddVectorToPosition(Position=$Particles.Position, Vector=<0 0 3>)
    Particles.RingScale:Vector3f=&VectorFromFloat(Value=0.25)
   ) {isSetParameters=true}
   /SetVariables(Particles.DiskScale:Vector3f=&Multiply_VectorByFloat(Vector=<1 1 0.5>, Float=1)) {isSetParameters=true}
   /SetVariables(
    Particles.GlowColor=&MakeLinearColorFromVectorAndFloat(
     "Vector (RGB)"=&Multiply_Vector(
      A=&MakeVectorFromLinearColorRGB(Color=&ColorFromCurve(
        ColorCurve=DataInterface() {propertyValues {
          AlphaCurve {keys []}
          BlueCurve {keys [<at 0.1172 0.807 RCIM_Cubic>,<at 0.259 0.8879>,<at 0.5248 0.019>,<at 0.6546 0.8879>,<at 0.8278 0.807 RCIM_Cubic>]}
          GreenCurve {keys [<at 0.1172 0.8879 RCIM_Cubic>,<at 0.259 0.3043>,<at 0.5248 1>,<at 0.6546 0>,<at 0.8278 0.8879 RCIM_Cubic>]}
          RedCurve {keys [<at 0.1172 0.0052 RCIM_Cubic>,<at 0.259 0>,<at 0.5248 0.1838>,<at 0.6546 0.4788>,<at 0.8278 0.0052 RCIM_Cubic>]}
         }}
        CurveIndex=&Multiply_Float(A=&Sine(Normalized Angle=$Emitter.Age, Period=2, Scale=1, Bias=0), B=$System.AudioScale)
       ))
      B=&VectorFromFloat(Value=1)
     )
     "Float (Alpha)"=0.3
    )
    Particles.GlowScale:Vector2f=&Vector2DFromFloat(Value=65)
   ) {isSetParameters=true}
   &ScaleColor(
    Scale Alpha=&FloatFromCurve(FloatCurve=DataInterface() {propertyValues {Curve {keys [<at 0 0>,<at 0.1 1>,<at 0.7 1>,<at 1 0>]}}})
    Scale RGB=&Multiply_Vector(
     A=&MakeVectorFromLinearColorRGB(Color=&ColorFromCurve(
       ColorCurve=DataInterface() {propertyValues {
         AlphaCurve {keys []}
         BlueCurve {keys [<at 0.1172 0.6802 RCIM_Cubic>,<at 0.259 0.8879>,<at 0.5248 0.019>,<at 0.6546 0.8879>,<at 0.8278 0.807 RCIM_Cubic>]}
         GreenCurve {keys [<at 0.1172 0.8879 RCIM_Cubic>,<at 0.259 0.3043>,<at 0.5248 1>,<at 0.6546 0>,<at 0.8278 0.8879 RCIM_Cubic>]}
         RedCurve {keys [<at 0.1172 0 RCIM_Cubic>,<at 0.259 0>,<at 0.5248 0.1838>,<at 0.6546 0.4788>,<at 0.8278 0.0052 RCIM_Cubic>]}
        }}
       CurveIndex=&Multiply_Float(A=&Sine(Normalized Angle=$Emitter.Age, Period=2, Scale=1, Bias=0), B=$System.AudioScale)
      ))
     B=&VectorFromFloat(Value=&Multiply_Float(
       A=&Multiply_Float(A=1, B=&FN_FrontEndFloatModifier(Gameplay=1, Front End=0.6))
       B=&FNTimeOfDayScalar(Night Scale=0.6, Morning Scale=1, Evening Scale=1)
      ))
    )
    ScaleA=true
    ScaleRGB=true
   ) {module=ScaleColor001}
  ]
  rendererClasses [&NiagaraMeshRendererProperties, &NiagaraSpriteRendererProperties]
  renderers [
   {rendererClass=&NiagaraMeshRendererProperties, rendererIndex=0}
   {rendererClass=&NiagaraMeshRendererProperties, rendererIndex=1}
   {rendererClass=&NiagaraMeshRendererProperties, rendererIndex=2}
   {rendererClass=&NiagaraSpriteRendererProperties, rendererIndex=3}
  ]
 }
 Loop_backboard {
  emitterUpdateScript [
   &EmitterState(Inactive Response=::NewEnumerator1, Life Cycle Mode=::NewEnumerator1, Loop Behavior=::NewEnumerator1, UseLoopDelay=true)
   &SpawnBurst_Instantaneous(Use Spawn Probability=false, UseLoopCountLimit=false)
  ]
  particleSpawnScript [
   &InitializeParticle(
    Lifetime=4.8
    Mesh Scale=<3 3 3>
    Mesh Scale Mode=::NewEnumerator4
    Position Offset=<0 0 125>
    Sprite Size=<220 180>
    Sprite Size Mode=::NewEnumerator4
    UsePositionOffset=true
    Write Material Random=true
   )
   &CameraOffset(Camera Offset Amount=-50)
   /SetVariables(Particles.Position002:NiagaraPosition=<0 0 10>) {isSetParameters=true}
   /SetVariables() {isSetParameters=true}
   &InitialMeshOrientation(Mesh Orientation Mode=::NewEnumerator2, Rotation=<0 0 0.25>, Use Rotation Vector=true)
  ]
  particleUpdateScript [
   &ParticleState(Kill Particles When Lifetime Has Elapsed=false, Let Infinitely Lived Particles Die When Emitter Deactivates=true)
   &ScaleSpriteSize(
    Uniform Curve Scale=&Multiply_Float(
     A=&Multiply_Float(A=1, B=&FN_PickFloatDependOnPlatform(TargetPlatformActive=1, TargetPlatformNoActive=0.9))
     B=&Multiply_Float(
      A=&FN_LOD_Distance_Remap(Start Distance=120, End Distance=600, Minimum Output=0.6, Maximum Output=1.2, Clamp Output=true)
      B=&FN_FrontEndFloatModifier(Gameplay=1, Front End=0.94)
     )
    )
    Uniform Curve Sprite Scale=DataInterface() {propertyValues {Curve {keys [{tangentMode=RCTM_User, time=0, value=1.1}, {tangentMode=RCTM_User, time=1, value=1.1}]}}}
   )
   &ScaleMeshSize(Scale Factor=&VectorFromFloat(Value=0.95))
   &ScaleColor(
    Scale Alpha=&FloatFromCurve(
     FloatCurve=DataInterface() {propertyValues {Curve {keys [<at 0 0>,<at 0.01 1>,<at 0.9 1>,<at 1 0>]}}}
     Scale Curve=&Multiply_Float(
      A=&Multiply_Float(A=0.7, B=&FN_FrontEndFloatModifier(Gameplay=1, Front End=0.5))
      B=&Multiply_Float(
       A=&FNTimeOfDayScalar(Night Scale=0.6, Morning Scale=1, Evening Scale=1)
       B=&FN_PickFloatDependOnPlatform(TargetPlatformActive=1, TargetPlatformNoActive=0.8)
      )
     )
    )
    Scale RGB=&MakeVectorFromLinearColorRGB(Color=<rgba 0.8 0.8 0.8 1>)
    ScaleA=true
    ScaleRGB=true
   )
   /SetVariables(Particles.Color002=&MakeLinearColorFromVectorAndFloat(
     "Vector (RGB)"=&MakeVectorFromLinearColorRGB(Color=&ColorFromCurve(ColorCurve=DataInterface() {propertyValues {
         AlphaCurve {keys []}
         BlueCurve {keys [<at 0.9413 0.8887>,<at 0.9861 1>]}
         GreenCurve {keys [<at 0.9413 1>,<at 0.9861 0.4111>]}
         RedCurve {keys [<at 0.9413 0>,<at 0.9861 0>]}
        }}))
     "Float (Alpha)"=&FloatFromCurve(
      FloatCurve=DataInterface() {propertyValues {Curve {keys [{arriveTangent=-3.8906, time=0, value=0}, <at 0.4 0.4>, <at 0.9 0.4>, <at 1 0>]}}}
      Scale Curve=&Multiply_Float(
       A=0.1
       B=&Multiply_Float(A=&FN_FrontEndFloatModifier(Gameplay=1, Front End=0.5), B=1)
      )
     )
    )) {isSetParameters=true}
   /SetVariables(Particles.Dyn002:Vector4f=&MakeVector4(
     X=0
     Y=0
     Z=&FloatFromCurve(FloatCurve=DataInterface() {propertyValues {Curve {keys [<at 0 0>,<at 0.08 1>]}}})
     W=0
    )) {isSetParameters=true}
   /SetVariables(Particles.Size003:Vector2f=&Multiply_Vector2D(
     A=&Vector2DFromFloat(Value=400)
     B=&Vector2DFromFloat(Value=&Multiply_Float(
       A=&FN_LOD_Distance_Remap(Start Distance=0, End Distance=1500, Minimum Output=0.8, Maximum Output=1, Clamp Output=true)
       B=1
      ))
    )) {isSetParameters=true}
   /SetVariables(Particles.Color003:LinearColor=&MakeLinearColorFromVectorAndFloat(
     "Vector (RGB)"=&MakeVectorFromLinearColorRGB(Color=<rgba 0 0.8 0.6504 1>)
     "Float (Alpha)"=&FloatFromCurve(
      FloatCurve=DataInterface() {propertyValues {Curve {keys [<at 0 0>,<at 0.1 1>,<at 0.9 1>,<at 1 0>]}}}
      Scale Curve=&Multiply_Float(
       A=0.2
       B=&Multiply_Float(A=&FN_FrontEndFloatModifier(Gameplay=1, Front End=0.6), B=1)
      )
     )
    )) {isSetParameters=true}
   &DynamicMaterialParameters(
    Index 0 Param 1=&FloatFromCurve(FloatCurve=DataInterface() {propertyValues {Curve {keys [{arriveTangent=-0.1184, interpMode=RCIM_Cubic, leaveTangent=-0.1184, tangentMode=RCTM_User, time=0, value=0.3}, <at 1 0.2>]}}})
    Index 0 Param 2=&FloatFromCurve(FloatCurve=DataInterface() {propertyValues {Curve {keys [{arriveTangent=-0.1866, interpMode=RCIM_Cubic, leaveTangent=-0.1866, tangentMode=RCTM_User, time=0, value=1}, <at 1 0.9>]}}})
    Index 0 Param 3=&FloatFromCurve(FloatCurve=DataInterface() {propertyValues {Curve {keys [<at 0 1>]}}})
    Module.Param0WriteEnabled=true
    Module.Param1WriteEnabled=true
    Module.Param2WriteEnabled=true
    Module.Param3WriteEnabled=false
   )
   /SetVariables(Particles.Dyn_glow:Vector4f=&MakeVector4(X=2, Y=5, Z=0, W=0)) {isSetParameters=true}
   &FN_Light_Attributes(
    Color=&ColorFromCurve(ColorCurve=DataInterface() {propertyValues {
       AlphaCurve {keys []}
       BlueCurve {keys [<at 0 0 RCIM_Cubic>,<at 0.1896 0.574>,<at 0.8328 0.6729>,<at 1 0 RCIM_Cubic>]}
       GreenCurve {keys [<at 0 0 RCIM_Cubic>,<at 0.1896 1>,<at 0.8328 1>,<at 1 0 RCIM_Cubic>]}
       RedCurve {keys [<at 0 0 RCIM_Cubic>,<at 0.1896 0>,<at 0.8328 0>,<at 1 0 RCIM_Cubic>]}
      }})
    Create Color Attribute=true
    Create Exponent Attribute=true
    Create Position Attribute=true
    Create Radius Attribute=true
    Exponent=3
    Position=<0 0 10>
    Radius=&Multiply_Float(A=360, B=&FN_FrontEndFloatModifier(Gameplay=1, Front End=0))
    Scale Color=&FloatFromCurve(
     FloatCurve=DataInterface() {propertyValues {Curve {keys [
         {arriveTangent=4.1988, time=0, value=0}
         {arriveTangent=-0.0105, time=0.05, value=1}
         {arriveTangent=-1.0448, leaveTangent=-1.0448, time=0.9, value=1}
         {arriveTangent=-5.4555, leaveTangent=-5.4555, time=1, value=0}
        ]}}}
     Scale Curve=&Multiply_Float(A=&FNTimeOfDayScalar(Night Scale=0.5, Morning Scale=1, Evening Scale=1), B=5)
    )
   )
  ]
  rendererClasses [
   &NiagaraSpriteRendererProperties
   &NiagaraMeshRendererProperties
   &NiagaraLightRendererProperties
  ]
  renderers [
   {rendererClass=&NiagaraSpriteRendererProperties}
   {rendererClass=&NiagaraMeshRendererProperties, rendererIndex=1}
   {rendererClass=&NiagaraSpriteRendererProperties, rendererIndex=2}
   {rendererClass=&NiagaraLightRendererProperties, rendererIndex=3}
  ]
 }
 Music_Ring {
  emitterUpdateScript [
   &EmitterState(Inactive Response=::NewEnumerator1, Life Cycle Mode=::NewEnumerator1, Loop Behavior=::NewEnumerator1, UseLoopDelay=false)
   &SpawnRate(
    SpawnRate=&FloatFromCurve(
     FloatCurve=DataInterface() {propertyValues {Curve {keys [<at 0 0 RCIM_Constant>,<at 0.2 1 RCIM_Constant>,<at 0.3 0 RCIM_Constant>,<at 0.47 1 RCIM_Constant>,<at 0.55 0 RCIM_Constant>,<at 0.72 0 RCIM_Constant>,<at 0.8 0 RCIM_Constant>,<at 1 0>]}}}
     CurveIndex=$Emitter.NormalizedLoopAge
     Scale Curve=5
    )
    Use Spawn Probability=false
   )
   &SpawnBurst_Instantaneous(Use Spawn Probability=true, UseLoopCountLimit=false)
   &SpawnBurst_Instantaneous(Spawn Group=1, Spawn Time=0.07, Use Spawn Probability=true, UseLoopCountLimit=false) {module=SpawnBurst_Instantaneous001}
   &SpawnBurst_Instantaneous(Spawn Count=2, Spawn Group=1, Spawn Time=2, Use Spawn Probability=true, UseLoopCountLimit=false) {enabled=false, module=SpawnBurst_Instantaneous002}
  ]
  particleSpawnScript [
   &FN_SetSpawnGroup(Number of Spawn Groups=2)
   &InitializeParticle(
    Color=<rgba 0 0.8854 0.4378 2>
    Lifetime Max=0.8
    Lifetime Min=0.7
    Lifetime Mode=::NewEnumerator1
    Mesh Scale Mode=::NewEnumerator3
    Mesh Uniform Scale=2.2
    Position Offset=<0 0 115>
    UsePositionOffset=true
    Write Material Random=true
   )
   &InitialMeshOrientation(
    Mesh Orientation Mode=::NewEnumerator2
    Rotation=&RandomRangeVector(Minimum=<0 0 0>, Maximum=<0 0 1>)
    Use Rotation Vector=true
   ) {module=InitialMeshOrientation001}
   &InitialMeshOrientation(
    Mesh Orientation Mode=::NewEnumerator2
    Rotation=&AddVectorToPosition(
     Position=&SelectVectorFromArray(Direct Array Index=$Particles.SpawnGroup, Vector Selection Array=DataInterface() {propertyValues {FloatData [<0 0.05 0>, <0 -0.05 0>]}})
     Vector=&RandomRangeVector(Minimum=<-0.01 -0.01 -0.01>, Maximum=<0.01 0.01 0.01>)
    )
    Use Rotation Vector=true
   )
  ]
  particleUpdateScript [
   &ParticleState()
   &ScaleMeshSize(Scale Factor=&VectorFromFloat(Value=&FloatFromCurve(
      FloatCurve=DataInterface() {propertyValues {Curve {keys [
          {arriveTangent=1.1517, interpMode=RCIM_Cubic, leaveTangent=1.1517, tangentMode=RCTM_User, time=0, value=0}
          {arriveTangent=1.2519, interpMode=RCIM_Cubic, leaveTangent=1.2519, tangentMode=RCTM_User, time=0.3136, value=0.7114}
          {arriveTangent=0.2998, interpMode=RCIM_Cubic, leaveTangent=0.2998, tangentMode=RCTM_User, time=0.9983, value=1.1136}
         ]}}}
      Scale Curve=&FN_LOD_Distance_Remap(Start Distance=250, End Distance=600, Minimum Output=0.5, Maximum Output=1.5, Clamp Output=true)
     )))
   &ScaleColor(
    Scale Alpha=&FloatFromCurve(
     FloatCurve=DataInterface() {propertyValues {Curve {keys [
         {arriveTangent=-29.315, interpMode=RCIM_Cubic, leaveTangent=-29.315, tangentMode=RCTM_User, time=0, value=8}
         {arriveTangent=-3.2846, interpMode=RCIM_Cubic, leaveTangent=-3.2846, tangentMode=RCTM_User, time=0.45, value=0.6}
         <at 1 0>
        ]}}}
     Scale Curve=&Multiply_Float(
      A=&FN_LOD_Distance_Remap(Start Distance=500, End Distance=1500, Minimum Output=0.6, Maximum Output=1, Clamp Output=true)
      B=&Multiply_Float(
       A=&FNTimeOfDayScalar(Night Scale=0.7, Morning Scale=1, Evening Scale=1)
       B=&Multiply_Float(A=&FN_FrontEndFloatModifier(Gameplay=1, Front End=0.8), B=1)
      )
     )
    )
    Scale RGB=&MakeVectorFromLinearColorRGB(Color=&ColorFromCurve(ColorCurve=DataInterface() {propertyValues {
        AlphaCurve {keys []}
        BlueCurve {keys [<at 0 5.6685>,<at 0.7108 10>,<at 0.9062 0.6975>]}
        GreenCurve {keys [<at 0 10>, {arriveTangent=-9.5462, time=0.7108, value=4.385}, <at 0.9062 1>]}
        RedCurve {keys [<at 0 0>,<at 0.7108 0>,<at 0.9062 0>]}
       }}))
    ScaleA=true
    ScaleRGB=true
   )
   /SetVariables(Particles.GlowColor=&MakeLinearColorFromVectorAndFloat(
     "Vector (RGB)"=&MakeVectorFromLinearColorRGB(Color=&ColorFromCurve(ColorCurve=DataInterface() {propertyValues {
         AlphaCurve {keys []}
         BlueCurve {keys [<at 0.4456 0.6819 RCIM_Cubic>,<at 0.8412 1>]}
         GreenCurve {keys [<at 0.4456 1 RCIM_Cubic>,<at 0.8412 0.7877>]}
         RedCurve {keys [<at 0.4456 0.2656 RCIM_Cubic>,<at 0.8412 0>]}
        }}))
     "Float (Alpha)"=&FloatFromCurve(
      FloatCurve=DataInterface() {propertyValues {Curve {keys [
          {arriveTangent=-1.8164, interpMode=RCIM_Cubic, leaveTangent=-1.8164, tangentMode=RCTM_User, time=0, value=0}
          {arriveTangent=-0.0446, interpMode=RCIM_Cubic, leaveTangent=-0.0446, tangentMode=RCTM_User, time=0.1, value=1}
          {arriveTangent=-1.9449, interpMode=RCIM_Cubic, leaveTangent=-1.9449, tangentMode=RCTM_User, time=0.55, value=0.4055}
          {arriveTangent=-0.1166, interpMode=RCIM_Cubic, leaveTangent=-0.1166, tangentMode=RCTM_User, time=1, value=0}
         ]}}}
      Scale Curve=&Multiply_Float(
       A=0.5
       B=&Multiply_Float(A=&FN_FrontEndFloatModifier(Gameplay=1, Front End=0.7), B=1)
      )
     )
    )) {isSetParameters=true, module=SetVariables_#3}
   &DynamicMaterialParameters(
    Index 0 Param 1=&FloatFromCurve(
     FloatCurve=DataInterface() {propertyValues {Curve {keys [<at 0.8 1>,<at 1 0.8>]}}}
     Scale Curve=&RandomRangeFloat(Minimum=4, Maximum=6)
    )
    Index 0 Param 2=&MakeCustomFloatFromBool(Bool=&RandomBool(), True Float=0.8, False Float=1)
    Index 0 Param 3=&MakeCustomFloatFromBool(Bool=&RandomBool(), True Float=0.8, False Float=1.2)
    Module.Index0Param0Write=true
    Module.Index0Param1Write=true
    Module.Index0Param2Write=true
    Module.Index0Param3Write=true
   )
   &ScaleColor(
    Scale Alpha=&FloatFromCurve(FloatCurve=DataInterface() {propertyValues {Curve {keys [<at 1 1>,<at 5 0>]}}}, CurveIndex=$Emitter.Age)
    ScaleA=true
    ScaleRGB=true
   ) {module=ScaleColor001}
  ]
  rendererClasses [&NiagaraMeshRendererProperties]
  renderers [
   {rendererClass=&NiagaraMeshRendererProperties, rendererIndex=0}
   {rendererClass=&NiagaraMeshRendererProperties, rendererIndex=1}
  ]
 }
 Triangle {
  emitterUpdateScript [
   &EmitterState(
    Inactive Response=::NewEnumerator1
    Life Cycle Mode=::NewEnumerator1
    Loop Behavior=::NewEnumerator1
    Loop Delay=0.6
    Loop Duration=4
    UseLoopDelay=true
   )
   &SpawnRate(SpawnRate=20, Use Spawn Probability=false)
   &SpawnRate(SpawnGroup=1, SpawnRate=20, Use Spawn Probability=false) {module=SpawnRate001}
  ]
  particleSpawnScript [
   &FN_SetSpawnGroup(Number of Spawn Groups=2)
   &InitializeParticle(
    Color=&SelectLinearColorFromArray(
     Array Sampling Mode=::NewEnumerator1
     Direct Array Index=&RandomRangeInt(Minimum=0, Maximum=2)
     Color Selection Array=DataInterface() {propertyValues {ColorData [<rgba 0.1394 0.2649 2 1>, <rgba 0 1.6981 2 1>, <rgba 1.4966 0.2713 2 1>]}}
     Random Seed=?
    )
    Lifetime Max=1
    Lifetime Min=0.2
    Lifetime Mode=::NewEnumerator1
    Mass Max=1.5
    Mass Mode=::NewEnumerator3
    Mesh Scale Mode=::NewEnumerator1
    Mesh Uniform Scale Max=0.45
    Mesh Uniform Scale Min=0.15
    Sprite Rotation Angle=&RandomRangeFloat(Minimum=-0.15, Maximum=0.15)
    Sprite Rotation Mode=::NewEnumerator1
    Sprite Size Mode=::NewEnumerator1
    Uniform Sprite Size Max=3.5
    Uniform Sprite Size Min=3
    UsePositionOffset=false
    Write Material Random=true
   )
   &ShapeLocation(
    Cylinder Height=200
    Cylinder Radius=60
    MaskSpawnGroup=true
    Offset=<80 0 100>
    Offset Mode=::NewEnumerator0
    Override Seed=false
    Shape Primitive=::NewEnumerator1
    Surface Only=false
    UseOwnerScale=true
   )
   &ShapeLocation(
    Cylinder Height=200
    Cylinder Radius=60
    MaskSpawnGroup=true
    Offset=<-80 0 100>
    Offset Mode=::NewEnumerator0
    Override Seed=false
    Shape Primitive=::NewEnumerator1
    Spawn Group Mask=1
    Surface Only=false
    UseOwnerScale=true
   ) {module=ShapeLocation001}
   &InitialMeshOrientation(
    Mesh Orientation Mode=::NewEnumerator2
    Rotation=&RandomRangeVector(Minimum=<0 0 0>, Maximum=<1 1 1>)
    Use Rotation Vector=true
   )
   &AddVelocity(Velocity=&RandomRangeVector(Minimum=<0 0 10>, Maximum=<0 0 20>))
   &ApplyInitialForces()
   /SetVariables(Particles.VisibilityTag=&RandomRangeInt(Minimum=0, Maximum=1)) {isSetParameters=true}
  ]
  particleUpdateScript [
   &ParticleState()
   &SpriteRotationRate(Rotation Rate=&Multiply_Float(
     A=&RandomRangeFloat(Minimum=-20, Maximum=20)
     B=&FloatFromCurve(FloatCurve=DataInterface() {propertyValues {Curve {keys [<at 0 1>,<at 1 0.45>]}}}, Scale Curve=10)
    ))
   &ScaleSpriteSize(
    Uniform Curve Scale=&RandomRangeFloat(Minimum=1, Maximum=1.5)
    Uniform Curve Sprite Scale=DataInterface() {propertyValues {Curve {keys [{tangentMode=RCTM_User, time=0, value=0.4}, <at 0.1 1.2>, <at 0.2 1>, {tangentMode=RCTM_User, time=1, value=1}]}}}
   )
   /SetVariables(Particles.Size003:Vector2f=&Multiply_Vector2D(
     A=&Vector2DFromFloat(Value=20)
     B=&Vector2DFromFloat(Value=&FloatFromCurve(FloatCurve=DataInterface() {propertyValues {Curve {keys [<at 0 0.6>,<at 0.2 1>,<at 0.8 1>,<at 1 0.6>]}}}))
    )) {isSetParameters=true}
   /SetVariables(Particles.Color003:LinearColor=&MakeLinearColorFromVectorAndFloat(
     "Vector (RGB)"=&MakeVectorFromLinearColorRGB(Color=$Particles.Color)
     "Float (Alpha)"=&FloatFromCurve(
      FloatCurve=DataInterface() {propertyValues {Curve {keys [<at 0 0 RCIM_Cubic>,<at 0.2 1 RCIM_Cubic>,<at 1 0 RCIM_Cubic>]}}}
      Scale Curve=&Multiply_Float(
       A=0.12
       B=&Multiply_Float(A=&FN_FrontEndFloatModifier(Gameplay=1, Front End=0.8), B=1)
      )
     )
    )) {isSetParameters=true}
   /SetVariables(Particles.Dyn002:Vector4f=&MakeVector4(X=2, Y=1, Z=0, W=0)) {isSetParameters=true}
   &UpdateMeshOrientation(
    Rotation Rate=&RandomRangeFloat(Minimum=0.5, Maximum=1.2)
    Rotation Vector=&RandomRangeVector(Minimum=<0.2 0.2 0.2>, Maximum=<0.6 0.6 0.6>)
   )
   &ScaleMeshSize(Scale Factor=&VectorFromFloat(Value=&FloatFromCurve(
      FloatCurve=DataInterface() {propertyValues {Curve {keys [
          {interpMode=RCIM_Cubic, tangentMode=RCTM_User, time=0, value=0}
          {interpMode=RCIM_Cubic, tangentMode=RCTM_User, time=0.2, value=1}
          {arriveTangent=-0.0093, interpMode=RCIM_Cubic, leaveTangent=-0.0093, tangentMode=RCTM_User, time=0.5, value=1}
          {tangentMode=RCTM_User, time=1, value=0}
         ]}}}
      Scale Curve=&Multiply_Float(
       A=&RandomRangeFloat(Minimum=0.8, Maximum=1.2)
       B=&ClampFloat(Float=$NPC.FortniteNPC.EventAudio_Amplitude, Min=0.4)
      )
     )))
   &DynamicMaterialParameters(
    Index 0 Param 1=&Add_Float(
     A=20
     B=&Lerp_Float(
      A=&Multiply_Float(A=$NPC.FortniteNPC.EventAudio_Amplitude, B=$NPC.FortniteNPC.EventAudio_Amplitude)
      B=1
      Alpha=$System.MusicReactivePreview
     )
    )
    Index 0 Param 2=5
    Index 0 Param 3=&RandomRangeFloat(Minimum=0, Maximum=0.2)
    Index 0 Param 4=&RandomRangeFloat(Minimum=0, Maximum=0.2)
    Module.Index0Param0Write=true
    Module.Index0Param1Write=true
    Module.Index0Param2Write=true
    Module.Index0Param3Write=true
   )
   &ScaleColor(
    Scale Alpha=&RandomRangeFloat(Minimum=1, Maximum=3)
    Scale RGB=&Multiply_Vector(
     A=&MakeVectorFromLinearColorRGB(Color=<rgba 1 1 1 1>)
     B=&VectorFromFloat(Value=&FloatFromCurve(FloatCurve=DataInterface() {propertyValues {Curve {keys [<at 0 1>,<at 0.1157 0.0721>,<at 0.228 0.9759>,<at 0.3274 0.0692>,<at 0.4199 0.8624>,<at 0.5141 0.3216>,<at 0.617 0.9504>,<at 0.7287 0.0762>,<at 0.8508 1.0029>,<at 1 0>]}}}))
    )
    ScaleA=true
    ScaleRGB=true
   )
   &SolveForcesAndVelocity(Clamp Velocity=false, Limit Acceleration=false)
  ]
  rendererClasses [&NiagaraMeshRendererProperties, &NiagaraSpriteRendererProperties]
  renderers [
   {rendererClass=&NiagaraMeshRendererProperties, rendererIndex=0}
   {rendererClass=&NiagaraSpriteRendererProperties, rendererIndex=1}
  ]
 }
 backboard_Line {
  emitterUpdateScript [
   &EmitterState(Inactive Response=::NewEnumerator1, Life Cycle Mode=::NewEnumerator1, Loop Behavior=::NewEnumerator1, Loop Delay=2.5, UseLoopDelay=true)
   &SpawnBurst_Instantaneous(Use Spawn Probability=false, UseLoopCountLimit=false)
  ]
  particleSpawnScript [
   &InitializeParticle(
    Color=<rgba 0 0.9851 1 1>
    Lifetime=2.1
    Mesh Scale Mode=::NewEnumerator3
    Mesh Uniform Scale=1.5
    Position Offset=<0 0 125>
    Sprite Size=<220 180>
    Sprite Size Mode=::NewEnumerator4
    UsePositionOffset=true
    Write Material Random=true
   )
   &CameraOffset(Camera Offset Amount=-50)
  ]
  particleUpdateScript [
   &ParticleState()
   &ScaleSpriteSize(
    Uniform Curve Scale=&FN_LOD_Distance_Remap(Start Distance=0, End Distance=1000, Minimum Output=0.7, Maximum Output=1.2, Clamp Output=true)
    Uniform Curve Sprite Scale=DataInterface() {propertyValues {Curve {keys [{tangentMode=RCTM_User, time=1, value=1}]}}}
   )
   &ScaleColor(
    Scale Alpha=&FloatFromCurve(FloatCurve=DataInterface() {propertyValues {Curve {keys [<at 0 0>,<at 0.3 1>,<at 0.7 1>,<at 1 0>]}}})
    Scale RGB=&Multiply_Vector(
     A=&MakeVectorFromLinearColorRGB(Color=&ColorFromCurve(ColorCurve=DataInterface() {propertyValues {
         AlphaCurve {keys []}
         BlueCurve {keys [<at 0.0479 0.7703 RCIM_Cubic>,<at 0.1132 1>,<at 0.2209 1>,<at 0.3303 0.0481>,<at 0.4571 0.2357>,<at 0.5606 1>,<at 0.6453 0.104>,<at 0.7414 0.7535>,<at 0.8333 1>,<at 1 0.6641 RCIM_Cubic>]}
         GreenCurve {keys [<at 0.0479 1 RCIM_Cubic>,<at 0.1132 0.1213>,<at 0.2209 0>,<at 0.3303 0>,<at 0.4571 1>,<at 0.5606 0>,<at 0.6453 0>,<at 0.7414 1>,<at 0.8333 0.136>,<at 1 1 RCIM_Cubic>]}
         RedCurve {keys [<at 0.0479 0 RCIM_Cubic>,<at 0.1132 0>,<at 0.2209 0.5626>,<at 0.3303 1>,<at 0.4571 0>,<at 0.5606 0.3794>,<at 0.6453 1>,<at 0.7414 0>,<at 0.8333 0>,<at 1 0 RCIM_Cubic>]}
        }}))
     B=&VectorFromFloat(Value=&FloatFromCurve(
       FloatCurve=DataInterface() {propertyValues {Curve {keys [<at 0 0>,<at 0.1 1.2>,<at 0.1325 0.3057>,<at 0.2 1.3>,<at 0.2908 0.1161>,<at 0.3536 1.0659>,<at 0.3958 0.2719>,<at 0.4652 0.8287>,<at 0.5266 0.3087>,<at 0.6 1.3>,<at 0.6509 0.1123>,<at 0.7026 0.9719>,<at 0.7559 0.0586>,<at 0.8 1.3>,<at 0.8722 0.0956>,<at 0.9 1>,<at 1 0>]}}}
       Scale Curve=30
      ))
    )
    ScaleA=true
    ScaleRGB=true
   )
  ]
  rendererClasses [&NiagaraSpriteRendererProperties]
  renderers [{rendererClass=&NiagaraSpriteRendererProperties}]
 }
 backboard_Ring {
  emitterUpdateScript [
   &EmitterState(
    Inactive Response=::NewEnumerator1
    Life Cycle Mode=::NewEnumerator1
    Loop Behavior=::NewEnumerator1
    Loop Delay=2.5
    Loop Duration=4
    UseLoopDelay=true
   )
   &SpawnBurst_Instantaneous(Use Spawn Probability=true, UseLoopCountLimit=false)
   &SpawnBurst_Instantaneous(Spawn Time=0.5, Use Spawn Probability=true, UseLoopCountLimit=false) {module=SpawnBurst_Instantaneous001}
   &SpawnBurst_Instantaneous(Spawn Time=1, Use Spawn Probability=true, UseLoopCountLimit=false) {module=SpawnBurst_Instantaneous002}
  ]
  particleSpawnScript [
   &InitializeParticle(
    Color=<rgba 0 0.8854 0.4828 2>
    Lifetime=1.5
    Mesh Scale Mode=::NewEnumerator3
    Mesh Uniform Scale=1.5
    Position Offset=<0 0 125>
    UsePositionOffset=true
    Write Material Random=true
   )
   &InitialMeshOrientation(
    Mesh Orientation Mode=::NewEnumerator2
    Rotation=&RandomRangeVector(Minimum=<0 0 0>, Maximum=<0 0 1>)
    Use Rotation Vector=true
   ) {module=InitialMeshOrientation001}
   &InitialMeshOrientation(Mesh Orientation Mode=::NewEnumerator2, Rotation=<0 0.25 0>, Use Rotation Vector=true)
   &CameraOffset(Camera Offset Amount=-50)
  ]
  particleUpdateScript [
   &ParticleState()
   &ScaleMeshSize(Scale Factor=&VectorFromFloat(Value=&FloatFromCurve(
      FloatCurve=DataInterface() {propertyValues {Curve {keys [
          {arriveTangent=1.3713, interpMode=RCIM_Cubic, leaveTangent=1.3713, tangentMode=RCTM_User, time=0, value=0}
          {arriveTangent=0.6824, interpMode=RCIM_Cubic, leaveTangent=0.6824, tangentMode=RCTM_User, time=1, value=1.2}
         ]}}}
      Scale Curve=&FN_LOD_Distance_Remap(Start Distance=240, End Distance=600, Minimum Output=0.75, Maximum Output=1.2, Clamp Output=true)
     )))
   &ScaleColor(
    Scale Alpha=&FloatFromCurve(
     FloatCurve=DataInterface() {propertyValues {Curve {keys [
         {arriveTangent=-11.284, interpMode=RCIM_Cubic, leaveTangent=-11.284, tangentMode=RCTM_User, time=0, value=5}
         {arriveTangent=-4.0733, interpMode=RCIM_Cubic, leaveTangent=-4.0733, tangentMode=RCTM_User, time=0.4, value=1}
         <at 0.85 0>
        ]}}}
     Scale Curve=&FN_LOD_Distance_Remap(Start Distance=0, End Distance=1200, Minimum Output=1, Maximum Output=1.5, Clamp Output=true)
    )
    Scale RGB=&MakeVectorFromLinearColorRGB(Color=&ColorFromCurve(ColorCurve=DataInterface() {propertyValues {
        AlphaCurve {keys []}
        BlueCurve {keys [<at 0 4.4847>,<at 0.1436 4.6207>,<at 0.9135 1>]}
        GreenCurve {keys [<at 0 10>, {arriveTangent=-9.5462, time=0.1436, value=10}, <at 0.9135 0.7171>]}
        RedCurve {keys [<at 0 0>,<at 0.1436 0>,<at 0.9135 0.0281>]}
       }}))
    ScaleA=true
    ScaleRGB=true
   )
   /SetVariables(Particles.GlowColor=&MakeLinearColorFromVectorAndFloat(
     "Vector (RGB)"=&MakeVectorFromLinearColorRGB(Color=&ColorFromCurve(ColorCurve=DataInterface() {propertyValues {
         AlphaCurve {keys []}
         BlueCurve {keys [<at 0.2366 0.5642 RCIM_Cubic>]}
         GreenCurve {keys [<at 0.2366 1 RCIM_Cubic>]}
         RedCurve {keys [<at 0.2366 0 RCIM_Cubic>]}
        }}))
     "Float (Alpha)"=&FloatFromCurve(
      FloatCurve=DataInterface() {propertyValues {Curve {keys [
          {arriveTangent=-1.8164, interpMode=RCIM_Cubic, leaveTangent=-1.8164, tangentMode=RCTM_User, time=0, value=0}
          {arriveTangent=-0.0446, interpMode=RCIM_Cubic, leaveTangent=-0.0446, tangentMode=RCTM_User, time=0.1, value=1}
          {arriveTangent=-1.9449, interpMode=RCIM_Cubic, leaveTangent=-1.9449, tangentMode=RCTM_User, time=0.5, value=0.4}
          {arriveTangent=-0.1445, interpMode=RCIM_Cubic, leaveTangent=-0.1445, tangentMode=RCTM_User, time=0.85, value=0}
         ]}}}
      Scale Curve=&Multiply_Float(
       A=0.5
       B=&Multiply_Float(A=&FN_FrontEndFloatModifier(Gameplay=1, Front End=0.2), B=1)
      )
     )
    )) {isSetParameters=true, module=SetVariables_#3}
   &DynamicMaterialParameters(
    Index 0 Param 1=&FloatFromCurve(
     FloatCurve=DataInterface() {propertyValues {Curve {keys [<at 0 1>, {arriveTangent=-6.1259, time=0.8, value=1}, <at 1 0>]}}}
     Scale Curve=&RandomRangeFloat(Minimum=5, Maximum=8)
    )
    Index 0 Param 2=&MakeCustomFloatFromBool(Bool=&RandomBool(), True Float=0.8, False Float=1)
    Index 0 Param 3=&MakeCustomFloatFromBool(Bool=&RandomBool(), True Float=0.8, False Float=1.2)
    Index 0 Param 4=&FloatFromCurve(FloatCurve=DataInterface() {propertyValues {Curve {keys [<at 0 0.5>,<at 0.6 1>]}}})
    Module.Index0Param0Write=true
    Module.Index0Param1Write=true
    Module.Index0Param2Write=true
    Module.Index0Param3Write=true
   )
   &ScaleColor(
    Scale Alpha=&FloatFromCurve(FloatCurve=DataInterface() {propertyValues {Curve {keys [<at 0 1>,<at 4.3 1>,<at 4.6 0>]}}}, CurveIndex=$Emitter.Age)
    ScaleA=true
    ScaleRGB=true
   ) {module=ScaleColor001}
  ]
  rendererClasses [&NiagaraMeshRendererProperties]
  renderers [
   {rendererClass=&NiagaraMeshRendererProperties, rendererIndex=0}
   {rendererClass=&NiagaraMeshRendererProperties, rendererIndex=1}
  ]
 }
 backboard_star {
  emitterUpdateScript [
   &EmitterState(
    Inactive Response=::NewEnumerator1
    Life Cycle Mode=::NewEnumerator1
    Loop Behavior=::NewEnumerator1
    Loop Delay=0.1
    Loop Duration=4
    UseLoopDelay=true
   )
   &SpawnRate(
    SpawnRate=&FloatFromCurve(FloatCurve=DataInterface() {propertyValues {Curve {keys [<at 0.55 1>,<at 0.6 0>]}}}, CurveIndex=$Emitter.NormalizedLoopAge, Scale Curve=2)
    Use Spawn Probability=false
   )
   &SpawnBurst_Instantaneous(Use Spawn Probability=false, UseLoopCountLimit=false)
  ]
  particleSpawnScript [
   &InitializeParticle(
    Color=<rgba 3 3 3 1>
    Position Offset=<0 0 125>
    Sprite Size=<200 180>
    Sprite Size Mode=::NewEnumerator4
    UsePositionOffset=true
    Write Material Random=true
   )
   &CameraOffset(Camera Offset Amount=-50)
  ]
  particleUpdateScript [
   &ParticleState()
   &ScaleSpriteSize(
    Uniform Curve Scale=&FN_LOD_Distance_Remap(Start Distance=240, End Distance=600, Minimum Output=0.65, Maximum Output=1.1, Clamp Output=true)
    Uniform Curve Sprite Scale=DataInterface() {propertyValues {Curve {keys [{arriveTangent=-0.0248, interpMode=RCIM_Cubic, leaveTangent=-0.0248, tangentMode=RCTM_User, time=0, value=1}]}}}
   )
   &ScaleColor(
    Scale Alpha=&FloatFromCurve(FloatCurve=DataInterface() {propertyValues {Curve {keys [<at 0 0 RCIM_Cubic>,<at 0.4 1 RCIM_Cubic>,<at 1 0 RCIM_Cubic>]}}}, Scale Curve=2)
    Scale RGB=&Multiply_Vector(
     A=&MakeVectorFromLinearColorRGB(Color=<rgba 1 1 1 1>)
     B=&VectorFromFloat(Value=&Multiply_Float(
       A=5
       B=&Multiply_Float(
        A=&FNTimeOfDayScalar(Night Scale=0.8, Morning Scale=1, Evening Scale=1)
        B=&FN_FrontEndFloatModifier(Gameplay=1, Front End=0.8)
       )
      ))
    )
    ScaleA=true
    ScaleRGB=true
   )
   /SetVariables(Particles.Color002=&Multiply_LinearColorByFloat(
     Linear Color=&MakeLinearColorFromVectorAndFloat(
      "Vector (RGB)"=&MakeVectorFromLinearColorRGB(Color=&ColorFromCurve(ColorCurve=DataInterface() {propertyValues {
          AlphaCurve {keys []}
          BlueCurve {keys [<at 0.0604 0.0534 RCIM_Cubic>,<at 0.6661 0.3088>,<at 0.8205 0.7>]}
          GreenCurve {keys [<at 0.0604 0 RCIM_Cubic>,<at 0.6661 0.7>,<at 0.8205 0.6679>]}
          RedCurve {keys [<at 0.0604 0.7 RCIM_Cubic>,<at 0.6661 0>,<at 0.8205 0>]}
         }}))
      "Float (Alpha)"=1.1
     )
     Float=&Multiply_Float(
      A=2
      B=&Multiply_Float(
       A=&FNTimeOfDayScalar(Night Scale=0.8, Morning Scale=1, Evening Scale=1)
       B=&Multiply_Float(
        A=&FN_FrontEndFloatModifier(Gameplay=1, Front End=0.8)
        B=&FN_PickFloatDependOnPlatform(TargetPlatformActive=1, TargetPlatformNoActive=0.7)
       )
      )
     )
    )) {isSetParameters=true}
   &DynamicMaterialParameters(
    Index 0 Param 1=&FloatFromCurve(FloatCurve=DataInterface() {propertyValues {Curve {keys [<at 0 -0.2>,<at 1 2.25>]}}})
    Index 0 Param 2=&FloatFromCurve(FloatCurve=DataInterface() {propertyValues {Curve {keys [<at 0 0.6>,<at 1 1>]}}})
    Module.Index0Param0Write=true
    Module.Index0Param1Write=true
    Module.Index0Param2Write=true
    Module.Index0Param3Write=true
   )
  ]
  rendererClasses [&NiagaraSpriteRendererProperties]
  renderers [
   {rendererClass=&NiagaraSpriteRendererProperties}
   {rendererClass=&NiagaraSpriteRendererProperties, rendererIndex=1}
  ]
 }
}
summary {
 systemName=NS_Emote_JadeTowelGloss_Searchlight
 emitters [
  {
   emitterName=Loop_StageLights
   rendererClasses [&NiagaraMeshRendererProperties, &NiagaraSpriteRendererProperties]
  }
  {emitterName=Triangle, rendererClasses [&NiagaraMeshRendererProperties, &NiagaraSpriteRendererProperties]}
  {emitterName=Box, rendererClasses [&NiagaraMeshRendererProperties]}
  {emitterName=BigStar, rendererClasses [&NiagaraSpriteRendererProperties]}
  {emitterName=Music_Ring, rendererClasses [&NiagaraMeshRendererProperties]}
  {emitterName=backboard_star, rendererClasses [&NiagaraSpriteRendererProperties]}
  {emitterName=backboard_Line, rendererClasses [&NiagaraSpriteRendererProperties]}
  {
   emitterName=Loop_backboard
   rendererClasses [
    &NiagaraSpriteRendererProperties
    &NiagaraMeshRendererProperties
    &NiagaraLightRendererProperties
   ]
  }
  {emitterName=BoxGlow, rendererClasses [&NiagaraMeshRendererProperties]}
  {emitterName=backboard_Ring, rendererClasses [&NiagaraMeshRendererProperties]}
 ]
 userVariables [
  User.bIsFrontEnd {type=NiagaraBool}
  User.bIsFrontEndPreview {type=NiagaraBool}
 ]
}
